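# Niagara deploy effect for the MissusMind glider: a backpack splash plus goo flung from the glider's skeletal mesh.
# BackpackSplash and GooBig/GooSmall burst from sampled bones or sockets, with goo following curve-scaled per-spawn-group vectors; GooSmall simulates on GPU.
# GooSmallWings emits on a timed rate curve across multiple bones.
# User.Variant drives renderer visibility tags and tint, while time-of-day curves and front-end modifiers feed material lighting parameters.
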NiagaraSystem /BRCosmetics/Gadgets/Assets/VinderTech_GliderChute/Glider_MissusMind/FX/NS_Glider_MissusMind_Deploy
dependencies {
 usedDataInterfaces [
  &NiagaraDataInterfaceColorCurve
  &NiagaraDataInterfaceCurve
  &NiagaraDataInterfaceVector2DCurve
  &NiagaraDataInterfaceArrayFloat3
  &NiagaraDataInterfaceArrayFloat
  &NiagaraDataInterfaceVectorCurve
  &NiagaraDataInterfacePlatformSet
  &NiagaraDataInterfaceArrayInt32
  &NiagaraDataInterfaceSkeletalMesh
 ]
 usedDynamicInputs [
  &MaskVector2dBySpawnGroup
  &RandomRangeVector2D
  &MakeFloatFromVector2d
  &SetBoolByFloatComparison
  &RemapRange
  &NormalizeVector
  &TransformVector
 ]
 usedModules [
  &self:SystemSpawnScript.SystemScriptSource.SystemScriptGraph.NiagaraNodeAssignment_0.SetVariables_#1
  &SystemState
  &Timeline
  &FN_Integrator
  &InterpolateOverTime
 ]
}
emitters {
 BackpackSplash {
  emitterUpdateScript [
   &EmitterState()
   &SpawnBurst_Instantaneous(Spawn Count=4, Use Spawn Probability=false, UseLoopCountLimit=false)
   /SetVariables(Emitter.EmitterRandom:NiagaraFloat=&RandomRangeFloat(Minimum=0, Maximum=1)) {isSetParameters=true}
  ]
  particleSpawnScript [
   &FN_SetSpawnGroup()
   &InitializeParticle(
    Lifetime=&MaskFloatBySpawnGroup(Masked Float=&RandomRangeFloat(Minimum=0.6, Maximum=0.75), Passthrough Float=0.64)
    Sprite Rotation Mode=::NewEnumerator2
    Sprite Size Mode=::NewEnumerator3
    Uniform Sprite Size=&Lerp_Float(A=&RandomRangeFloat(Minimum=4.5, Maximum=5), B=3.5, Alpha=$User.Variant)
    UsePositionOffset=false
    Write Material Random=true
   )
   &CameraOffset(Camera Offset Amount=14)
   /SetVariables(Particles.VisibilityTag=&MakeIntFromFloat(Float=&Lerp_Float(
      A=&SelectFloatFromArray(Direct Array Index=&SpawnGroup(), Float Selection Array=DataInterface() {propertyValues {FloatData [0,1,5,5]}})
      B=&SelectFloatFromArray(Direct Array Index=&SpawnGroup(), Float Selection Array=DataInterface() {propertyValues {FloatData [4,4,2,3]}})
      Alpha=$User.Variant
     ))) {isSetParameters=true}
   /SetVariables(Particles.SplashVariantCamOffset=16) {isSetParameters=true}
  ]
  particleUpdateScript [
   &ParticleState()
   &ScaleColor(
    Scale RGB=&MakeVectorFromLinearColorRGB(Color=&LerpLinearColors(
      StartColor=&LerpLinearColors(
       StartColor=<rgba 1 1 1 9>
       EndColor=<rgba 0.28 0.28 0.28 9>
       LerpFactor=&FN_FrontEndFloatModifier(Gameplay=0, Front End=1)
      )
      EndColor=&LerpLinearColors(
       StartColor=<rgba 0.61 0.61 0.61 9>
       EndColor=<rgba 0.61 0.61 0.61 9>
       LerpFactor=&FN_FrontEndFloatModifier(Gameplay=0, Front End=1)
      )
      LerpFactor=$User.Variant
     ))
    ScaleA=true
    ScaleRGB=true
   )
   &ScaleSpriteSize(
    Uniform Curve Scale=29
    Uniform Curve Sprite Scale=DataInterface() {propertyValues {Curve {keys [
        {arriveTangent=1.1867, interpMode=RCIM_Cubic, leaveTangent=1.1867, tangentMode=RCTM_User, time=0, value=0.4}
        {interpMode=RCIM_Cubic, leaveTangent=-1.3651, time=1, value=1}
       ]}}}
   ) {module=ScaleSpriteSize001}
   &DynamicMaterialParameters(
    Index 0 Param 1=1
    Index 0 Param 2=1
    Index 0 Param 3=&FloatFromCurve(FloatCurve=DataInterface() {propertyValues {Curve {keys [
         {arriveTangent=0.5904, interpMode=RCIM_Cubic, leaveTangent=0.5904, tangentMode=RCTM_User, time=0, value=0}
         {arriveTangent=2.1055, interpMode=RCIM_Cubic, leaveTangent=2.1055, tangentMode=RCTM_User, time=1, value=1}
        ]}}})
    Index 0 Param 4=&RandomRangeFloat(Minimum=0, Maximum=33)
    Index 1 Param 1=&FloatFromCurve(
     FloatCurve=DataInterface() {propertyValues {Curve {keys [
         <at 0.216 1 RCIM_Cubic>
         {interpMode=RCIM_Cubic, tangentMode=RCTM_User, time=0.3505, value=0}
         {interpMode=RCIM_Cubic, tangentMode=RCTM_User, time=0.71, value=0}
         <at 0.8425 1 RCIM_Cubic>
        ]}}}
     CurveIndex=Linked() {linkedVariable {name=NPC.FortniteNPC.TimeOfDay_0-1, type=NiagaraFloat}}
    )
    Index 1 Param 2=&FloatFromCurve(
     FloatCurve=DataInterface() {propertyValues {Curve {keys [
         <at 0.216 0.4 RCIM_Cubic>
         {interpMode=RCIM_Cubic, tangentMode=RCTM_User, time=0.3505, value=0.94}
         {interpMode=RCIM_Cubic, tangentMode=RCTM_User, time=0.71, value=0.94}
         <at 0.8425 0.4 RCIM_Cubic>
        ]}}}
     CurveIndex=Linked() {linkedVariable {name=NPC.FortniteNPC.TimeOfDay_0-1, type=NiagaraFloat}}
    )
    Module.Index0Param0Write=true
    Module.Index0Param1Write=true
    Module.Index0Param2Write=true
    Module.Index0Param3Write=true
    Module.Index1Param0Write=true
    Module.Index1Param1Write=true
    Module.Index1Param2Write=true
    Module.Index1Param3Write=true
    Write Parameter Index 1=true
   )
   &SkeletalMeshLocation(
    Bone / Socket Index=16
    Bone Sampling Mode=::NewEnumerator4
    MaskSpawnGroup=false
    Sampled Position Offset=&SelectVectorFromArray(Direct Array Index=&SpawnGroup(), Vector Selection Array=DataInterface() {propertyValues {FloatData [<0 0 8>]}})
    Skeletal Mesh=$System.Skeletal Mesh
   )
   /SetVariables(
    Particles.SplashColor:LinearColor=&ColorFromCurve(ColorCurve=DataInterface() {propertyValues {
       AlphaCurve {keys [<at 0.7708 1 RCIM_Cubic>,<at 1 0 RCIM_Cubic>]}
       BlueCurve {keys [<at 0 1 RCIM_Cubic>,<at 1 1 RCIM_Cubic>]}
       GreenCurve {keys [<at 0 1 RCIM_Cubic>,<at 1 1 RCIM_Cubic>]}
       RedCurve {keys [<at 0 1 RCIM_Cubic>,<at 1 1 RCIM_Cubic>]}
      }})
    Particles.SplashSize:Vector2f=&Vector2DFromFloat(Value=&FloatFromCurve(
      FloatCurve=DataInterface() {propertyValues {Curve {keys [{arriveTangent=1.3554, interpMode=RCIM_Cubic, leaveTangent=1.3554, tangentMode=RCTM_User, time=0, value=0.3}, <at 1 1>]}}}
      Scale Curve=&FN_FrontEndFloatModifier(Gameplay=115, Front End=90)
     ))
   ) {isSetParameters=true}
   /SetVariables(
    Particles.SplashDynamicPar:Vector4f=&MakeVector4(
     X=$Emitter.EmitterRandom
     Y=0
     Z=&FloatFromCurve(FloatCurve=DataInterface() {propertyValues {Curve {keys [
          {interpMode=RCIM_Cubic, tangentMode=RCTM_Break, time=0.3, value=-0.6}
          {interpMode=RCIM_Cubic, tangentMode=RCTM_Break, time=0.95, value=-1.3}
         ]}}})
     W=&FloatFromCurve(FloatCurve=DataInterface() {propertyValues {Curve {keys [{arriveTangent=2.6651, interpMode=RCIM_Cubic, leaveTangent=2.6651, tangentMode=RCTM_User, time=0, value=0.1}, <at 0.6 1.333 RCIM_Cubic>]}}})
    )
    Particles.SplashVariantColor:LinearColor=&ColorFromCurve(ColorCurve=DataInterface() {propertyValues {
       AlphaCurve {keys [<at 0 0 RCIM_Cubic>,<at 0.3203 0.4>,<at 1 1 RCIM_Cubic>]}
       BlueCurve {keys [<at 0 1 RCIM_Cubic>,<at 1 1 RCIM_Cubic>]}
       GreenCurve {keys [<at 0 1 RCIM_Cubic>,<at 1 1 RCIM_Cubic>]}
       RedCurve {keys [<at 0 1 RCIM_Cubic>,<at 1 1 RCIM_Cubic>]}
      }})
   ) {isSetParameters=true}
   /SetVariables(
    Particles.GooVariantDynamic:Vector4f=&MakeVector4(
     X=1
     Y=1
     Z=&FloatFromCurve(FloatCurve=DataInterface() {propertyValues {Curve {keys [
          <at 0 0>
          {arriveTangent=0.5886, interpMode=RCIM_Cubic, leaveTangent=0.5886, tangentMode=RCTM_User, time=0.1601, value=0}
          {arriveTangent=0.7127, tangentMode=RCTM_User, time=1, value=1}
         ]}}})
     W=&RandomRangeFloat(Minimum=0, Maximum=1)
    )
    Particles.Vector 2D:Vector2f=&Vector2DFromFloat(Value=&FloatFromCurve(
      FloatCurve=DataInterface() {propertyValues {Curve {keys [
          {arriveTangent=1.2382, interpMode=RCIM_Cubic, leaveTangent=1.2382, tangentMode=RCTM_User, time=0, value=0.2064}
          {interpMode=RCIM_Cubic, leaveTangent=-1.3651, time=1, value=1}
         ]}}}
      Scale Curve=17
     ))
   ) {isSetParameters=true}
   /SetVariables(Particles.RimLightDynamic:Vector4f=&MakeVector4_FromLinearColor(LinearColor=&LerpLinearColors(
      StartColor=&MultiplyLinearColor(
       Linear Color=&ColorFromCurve(
        ColorCurve=DataInterface() {propertyValues {
          AlphaCurve {keys [<at 0 1>,<at 0.125 1>,<at 0.25 1>,<at 0.375 1>,<at 0.5 1>,<at 0.625 1>,<at 0.75 1>,<at 0.875 1>,<at 1 1>]}
          BlueCurve {keys [<at 0 0.01>,<at 0.041 0.01>,<at 0.083 0.01>,<at 0.125 0.01>,<at 0.166 0.0077>,<at 0.208 0.0045>,<at 0.25 0.0092>,<at 0.291 0.0136>,<at 0.333 0.03>,<at 0.375 0.04>,<at 0.416 0.0525>,<at 0.458 0.0525>,<at 0.5 0.0525>,<at 0.541 0.0525>,<at 0.583 0.0525>,<at 0.625 0.04>,<at 0.666 0.03>,<at 0.708 0.0152>,<at 0.75 0.0153>,<at 0.791 0.0079>,<at 0.833 0.0025>,<at 0.875 0.0031>,<at 0.916 0.01>,<at 0.958 0.01>,<at 1 0.01>]}
          GreenCurve {keys [<at 0 0.0068>,<at 0.041 0.0064>,<at 0.083 0.0064>,<at 0.125 0.0071>,<at 0.166 0.0056>,<at 0.208 0.0041>,<at 0.25 0.0103>,<at 0.291 0.0186>,<at 0.333 0.0156>,<at 0.375 0.0208>,<at 0.416 0.0272>,<at 0.458 0.0272>,<at 0.5 0.0272>,<at 0.541 0.0272>,<at 0.583 0.0272>,<at 0.625 0.0208>,<at 0.666 0.0156>,<at 0.708 0.0192>,<at 0.75 0.0153>,<at 0.791 0.0092>,<at 0.833 0.0051>,<at 0.875 0.0062>,<at 0.916 0.0064>,<at 0.958 0.0063>,<at 1 0.0068>]}
          RedCurve {keys [<at 0 0.0033>,<at 0.041 0.0029>,<at 0.083 0.0032>,<at 0.125 0.0036>,<at 0.166 0.01>,<at 0.208 0.01>,<at 0.25 0.02>,<at 0.291 0.03>,<at 0.333 0.0086>,<at 0.375 0.0115>,<at 0.416 0.0151>,<at 0.458 0.0151>,<at 0.5 0.0151>,<at 0.541 0.0151>,<at 0.583 0.0151>,<at 0.625 0.0115>,<at 0.666 0.0086>,<at 0.708 0.03>,<at 0.75 0.03>,<at 0.791 0.02>,<at 0.833 0.01>,<at 0.875 0.01>,<at 0.916 0.0022>,<at 0.958 0.0019>,<at 1 0.0033>]}
         }}
        CurveIndex=Linked() {linkedVariable {name=NPC.FortniteNPC.TimeOfDay_0-1, type=NiagaraFloat}}
       )
       Scale Factor=&MakeFloat4FromVectorAndFloat(
        "Vector (RGB)"=&VectorFromFloat(Value=&FN_PickFloatDependOnPlatform(TargetPlatform=DataInterface() {propertyValues {Platforms {qualityLevelMask=-29}}}, TargetPlatformActive=0.2, TargetPlatformNoActive=1))
        "Float (Alpha)"=1
       )
      )
      EndColor=<rgba 0.0151 0.0272 0.0525 1>
      LerpFactor=&FN_FrontEndFloatModifier(Gameplay=0, Front End=1)
     ))) {isSetParameters=true}
   /SetVariables(Particles.HighlightDynamic:Vector4f=&MakeVector4_FromLinearColor(LinearColor=&LerpLinearColors(
      StartColor=&ColorFromCurve(
       ColorCurve=DataInterface() {propertyValues {
         AlphaCurve {keys [<at 0 1>,<at 0.125 1>,<at 0.25 1>,<at 0.375 1>,<at 0.5 1>,<at 0.625 1>,<at 0.75 1>,<at 0.875 1>,<at 1 1>]}
         BlueCurve {keys [<at 0 0.9216>,<at 0.041 0.8963>,<at 0.083 0.8308>,<at 0.125 0.7835>,<at 0.1865 0.2611>,<at 0.25 0.3013>,<at 0.333 0.3128>,<at 0.4094 0.2806>,<at 0.4889 0.2616>,<at 0.583 0.2814>,<at 0.666 0.2801>,<at 0.75 0.2471>,<at 0.791 0.2118>,<at 0.833 0.2338>,<at 0.875 0.2473>,<at 0.916 0.8388>,<at 0.958 0.9047>,<at 1 0.9216>]}
         GreenCurve {keys [<at 0 0.5616>,<at 0.041 0.5276>,<at 0.083 0.4795>,<at 0.125 0.5099>,<at 0.1865 0.2253>,<at 0.25 0.3543>,<at 0.333 0.4809>,<at 0.4094 0.4334>,<at 0.4889 0.389>,<at 0.583 0.2814>,<at 0.666 0.4785>,<at 0.75 0.2471>,<at 0.791 0.2686>,<at 0.833 0.4442>,<at 0.875 0.5188>,<at 0.916 0.5292>,<at 0.958 0.5688>,<at 1 0.6308>]}
         RedCurve {keys [<at 0 0.1584>,<at 0.041 0.1774>,<at 0.083 0.1644>,<at 0.125 0.1755>,<at 0.1865 0.6867>,<at 0.25 0.8148>,<at 0.333 0.8963>,<at 0.4094 0.9131>,<at 0.4889 0.9131>,<at 0.583 0.8714>,<at 0.666 0.8963>,<at 0.75 0.7529>,<at 0.791 0.7529>,<at 0.833 0.8469>,<at 0.875 0.8632>,<at 0.916 0.1791>,<at 0.958 0.1696>,<at 1 0.305>]}
        }}
       CurveIndex=Linked() {linkedVariable {name=NPC.FortniteNPC.TimeOfDay_0-1, type=NiagaraFloat}}
      )
      EndColor=<rgba 0.7 0.7 0.7 1>
      LerpFactor=&FN_FrontEndFloatModifier(Gameplay=0, Front End=1)
     ))) {isSetParameters=true}
  ]
  rendererClasses [&NiagaraSpriteRendererProperties]
  renderers [
   {rendererClass=&NiagaraSpriteRendererProperties}
   {rendererClass=&NiagaraSpriteRendererProperties, rendererIndex=1}
   {rendererClass=&NiagaraSpriteRendererProperties, rendererIndex=2}
   {rendererClass=&NiagaraSpriteRendererProperties, rendererIndex=3}
  ]
 }
 GooBig {
  emitterUpdateScript [
   &EmitterState()
   &SpawnBurst_Instantaneous(Spawn Count=18, Use Spawn Probability=false, UseLoopCountLimit=false)
  ]
  particleSpawnScript [
   &FN_SetSpawnGroup(Number of Spawn Groups=6)
   &InitializeParticle(
    Lifetime=&RandomRangeFloat(Minimum=0.55, Maximum=0.7)
    Sprite Size Mode=::NewEnumerator1
    Uniform Sprite Size Max=5.2
    Uniform Sprite Size Min=3
    UsePositionOffset=false
    Write Material Random=true
   )
   &CameraOffset(Camera Offset Amount=2)
   /SetVariables(Particles.RandFloat:NiagaraFloat=&RandomRangeFloat(Minimum=0.2, Maximum=1)) {isSetParameters=true}
   /SetVariables(Particles.VisibilityTag=&MakeIntFromFloat(Float=$User.Variant)) {isSetParameters=true}
  ]
  particleUpdateScript [
   &ParticleState()
   &ScaleColor(
    Scale RGB=&MakeVectorFromLinearColorRGB(Color=&LerpLinearColors(
      StartColor=&LerpLinearColors(
       StartColor=<rgba 1 1 1 9>
       EndColor=<rgba 0.28 0.28 0.28 9>
       LerpFactor=&FN_FrontEndFloatModifier(Gameplay=0, Front End=1)
      )
      EndColor=&LerpLinearColors(
       StartColor=<rgba 0.61 0.61 0.61 9>
       EndColor=<rgba 0.61 0.61 0.61 9>
       LerpFactor=&FN_FrontEndFloatModifier(Gameplay=0, Front End=1)
      )
      LerpFactor=$User.Variant
     ))
    ScaleA=true
    ScaleRGB=true
   )
   &ScaleSpriteSize(
    Uniform Curve Scale=&Lerp_Float(A=14, B=14, Alpha=$User.Variant)
    Uniform Curve Sprite Scale=DataInterface() {propertyValues {Curve {keys [
        {arriveTangent=-0.1897, interpMode=RCIM_Cubic, leaveTangent=-0.1897, leaveTangentWeight=0.3393, tangentMode=RCTM_User, time=0, value=0.4}
        {
         arriveTangent=0.6
         arriveTangentWeight=0.2266
         interpMode=RCIM_Cubic
         leaveTangent=0.6
         leaveTangentWeight=0.1067
         tangentMode=RCTM_User
         time=0.6799
         value=0.8952
        }
        {arriveTangentWeight=0.1067, interpMode=RCIM_Cubic, tangentMode=RCTM_User, time=1, value=1}
       ]}}}
   ) {module=ScaleSpriteSize001}
   &ScaleSpriteSize(
    Scale Factor=&Lerp_Vector2(
     A=&Vector2DFromCurve(
      Vector2Curve=DataInterface() {propertyValues {
        XCurve {keys [{tangentMode=RCTM_User, time=0, value=1}]}
        YCurve {keys [
          {interpMode=RCIM_Cubic, leaveTangentWeight=0.1747, tangentMode=RCTM_User, time=0, value=1.5848}
          {arriveTangentWeight=0.1747, interpMode=RCIM_Cubic, leaveTangentWeight=0.1455, tangentMode=RCTM_User, time=0.524, value=2.3131}
          {
           arriveTangent=-1.5632
           arriveTangentWeight=0.1455
           interpMode=RCIM_Cubic
           leaveTangent=-1.5632
           leaveTangentWeight=0.0132
           tangentMode=RCTM_User
           time=0.9604
           value=1.934
          }
         ]}
       }}
      ScaleCurve=<0.8 1.3>
     )
     B=&Vector2DFromCurve(
      Vector2Curve=DataInterface() {propertyValues {
        XCurve {keys [{tangentMode=RCTM_User, time=0, value=1}]}
        YCurve {keys [
          {interpMode=RCIM_Cubic, leaveTangentWeight=0.1313, tangentMode=RCTM_User, time=0, value=1}
          {arriveTangentWeight=0.1313, interpMode=RCIM_Cubic, leaveTangentWeight=0.1354, tangentMode=RCTM_User, time=0.3938, value=1.2688}
          {arriveTangentWeight=0.1354, interpMode=RCIM_Cubic, leaveTangentWeight=0.0667, tangentMode=RCTM_User, time=0.8, value=1.2076}
         ]}
       }}
      ScaleCurve=<0.8 1.3>
     )
     Alpha=&RandomRangeFloat(Minimum=0, Maximum=1)
    )
    Scale Sprite Size Mode=::NewEnumerator2
   )
   &DynamicMaterialParameters(
    Index 0 Param 1=1
    Index 0 Param 2=1
    Index 0 Param 3=&FloatFromCurve(FloatCurve=DataInterface() {propertyValues {Curve {keys [
         <at 0 0>
         {arriveTangent=0.44, interpMode=RCIM_Cubic, leaveTangent=0.44, time=0.4152, value=0.014}
         {arriveTangent=1.6861, interpMode=RCIM_Cubic, leaveTangent=1.6861, time=0.8228, value=0.3621}
         {arriveTangent=6.8718, time=1, value=1}
        ]}}})
    Index 0 Param 4=&RandomRangeFloat(Minimum=0, Maximum=33)
    Index 1 Param 1=&FloatFromCurve(
     FloatCurve=DataInterface() {propertyValues {Curve {keys [
         <at 0.216 1 RCIM_Cubic>
         {interpMode=RCIM_Cubic, tangentMode=RCTM_User, time=0.3505, value=0}
         {interpMode=RCIM_Cubic, tangentMode=RCTM_User, time=0.71, value=0}
         <at 0.8425 1 RCIM_Cubic>
        ]}}}
     CurveIndex=Linked() {linkedVariable {name=NPC.FortniteNPC.TimeOfDay_0-1, type=NiagaraFloat}}
    )
    Index 1 Param 2=&FloatFromCurve(
     FloatCurve=DataInterface() {propertyValues {Curve {keys [
         <at 0.216 0.4 RCIM_Cubic>
         {interpMode=RCIM_Cubic, tangentMode=RCTM_User, time=0.3505, value=0.94}
         {interpMode=RCIM_Cubic, tangentMode=RCTM_User, time=0.71, value=0.94}
         <at 0.8425 0.4 RCIM_Cubic>
        ]}}}
     CurveIndex=Linked() {linkedVariable {name=NPC.FortniteNPC.TimeOfDay_0-1, type=NiagaraFloat}}
    )
    Module.Index0Param0Write=true
    Module.Index0Param1Write=true
    Module.Index0Param2Write=true
    Module.Index0Param3Write=true
    Module.Index1Param0Write=true
    Module.Index1Param1Write=true
    Module.Index1Param2Write=true
    Module.Index1Param3Write=true
    Write Parameter Index 1=true
   )
   &SkeletalMeshLocation(
    Bone / Socket Index=&Add_Integer(A=&SpawnGroup(), B=10)
    Bone Sampling Mode=::NewEnumerator4
    MaskSpawnGroup=false
    Sampled Position Offset=&SelectVectorFromArray(Direct Array Index=&SpawnGroup(), Vector Selection Array=DataInterface() {propertyValues {FloatData [<0 0 0>]}})
    Skeletal Mesh=$System.Skeletal Mesh
   )
   /SetVariables(Particles.PosVector:Vector3f=&Add_Vector(
     A=&Multiply_Vector(
      A=&MakeVector(
       X=&RandomRangeFloat(
        Minimum=&SelectFloatFromArray(
         Array Sampling Mode=::NewEnumerator0
         Direct Array Index=?
         Float Selection Array=DataInterface() {propertyValues {FloatData [0.8,1,0.8,1,0.8]}}
         Random Seed=0
        )
        Maximum=1
       )
       Y=&Multiply_Float(A=1, B=$Particles.RandFloat)
       Z=&Multiply_Float(A=1.2, B=$Particles.RandFloat)
      )
      B=&SelectVectorFromArray(
       Direct Array Index=&SpawnGroup()
       Vector Selection Array=DataInterface() {propertyValues {FloatData [<33 -50 -50>, <5 -200 -80>, <-20 0 -70>, <33 -50 50>, <5 -200 80>, <-20 0 70>]}}
      )
     )
     B=<0 0 0>
    )) {isSetParameters=true, module=SetVariables_#2}
   /SetVariables(Particles.Position=&AddVectorToPosition(
     Position=$Particles.Position
     Vector=&Add_Vector(
      A=&VectorFromCurve(
       VectorCurve=DataInterface() {propertyValues {
         XCurve {keys [{arriveTangent=1.4099, interpMode=RCIM_Cubic, leaveTangent=1.4099, tangentMode=RCTM_User, time=0, value=0}, <at 1 1.3292 RCIM_Cubic>]}
         YCurve {keys [
           {arriveTangent=0.0186, interpMode=RCIM_Cubic, leaveTangent=0.0186, tangentMode=RCTM_User, time=0, value=0}
           {arriveTangent=0.3091, interpMode=RCIM_Cubic, leaveTangent=0.3091, tangentMode=RCTM_User, time=1, value=0.2612}
          ]}
         ZCurve {keys [{arriveTangent=2.9146, interpMode=RCIM_Cubic, leaveTangent=2.9146, tangentMode=RCTM_User, time=0, value=0}, <at 1 1.4648 RCIM_Cubic>]}
        }}
       Scale Curve=$Particles.PosVector
      )
      B=<0 0 0>
     )
    )) {isSetParameters=true, module=SetVariables_#3}
   &SpriteFacingAndAlignment(
    Sprite Alignment=&AddVectorToPosition(
     Position=$Particles.Position
     Vector=&Add_Vector(
      A=&VectorFromCurve(
       VectorCurve=DataInterface() {propertyValues {
         XCurve {keys [{arriveTangent=1.4099, interpMode=RCIM_Cubic, leaveTangent=1.4099, tangentMode=RCTM_User, time=0, value=0}, <at 1 1.3292 RCIM_Cubic>]}
         YCurve {keys [
           {arriveTangent=0.0186, interpMode=RCIM_Cubic, leaveTangent=0.0186, tangentMode=RCTM_User, time=0, value=0}
           {arriveTangent=0.3091, interpMode=RCIM_Cubic, leaveTangent=0.3091, tangentMode=RCTM_User, time=1, value=0.2612}
          ]}
         ZCurve {keys [{arriveTangent=2.9146, interpMode=RCIM_Cubic, leaveTangent=2.9146, tangentMode=RCTM_User, time=0, value=0}, <at 1 1.4648 RCIM_Cubic>]}
        }}
       Scale Curve=$Particles.PosVector
      )
      B=<0 0 0>
     )
    )
    WriteAlignment=true
    WriteFacing=false
   )
   /SetVariables(Particles.VariantDynamic:Vector4f=&MakeVector4(
     X=1
     Y=1
     Z=&FloatFromCurve(FloatCurve=DataInterface() {propertyValues {Curve {keys [
          <at 0 0>
          {arriveTangent=1, interpMode=RCIM_Cubic, leaveTangent=1, time=0.3628, value=0}
          {arriveTangent=3.3266, tangentMode=RCTM_User, time=1, value=1}
         ]}}})
     W=&RandomRangeFloat(Minimum=0, Maximum=1)
    )) {isSetParameters=true}
   /SetVariables(Particles.RimLightDynamic:Vector4f=&MakeVector4_FromLinearColor(LinearColor=&LerpLinearColors(
      StartColor=&MultiplyLinearColor(
       Linear Color=&ColorFromCurve(
        ColorCurve=DataInterface() {propertyValues {
          AlphaCurve {keys [<at 0 1>,<at 0.125 1>,<at 0.25 1>,<at 0.375 1>,<at 0.5 1>,<at 0.625 1>,<at 0.75 1>,<at 0.875 1>,<at 1 1>]}
          BlueCurve {keys [<at 0 0.01>,<at 0.041 0.01>,<at 0.083 0.01>,<at 0.125 0.01>,<at 0.166 0.0077>,<at 0.208 0.0045>,<at 0.25 0.0092>,<at 0.291 0.0136>,<at 0.333 0.03>,<at 0.375 0.04>,<at 0.416 0.0525>,<at 0.458 0.0525>,<at 0.5 0.0525>,<at 0.541 0.0525>,<at 0.583 0.0525>,<at 0.625 0.04>,<at 0.666 0.03>,<at 0.708 0.0152>,<at 0.75 0.0153>,<at 0.791 0.0079>,<at 0.833 0.0025>,<at 0.875 0.0031>,<at 0.916 0.01>,<at 0.958 0.01>,<at 1 0.01>]}
          GreenCurve {keys [<at 0 0.0068>,<at 0.041 0.0064>,<at 0.083 0.0064>,<at 0.125 0.0071>,<at 0.166 0.0056>,<at 0.208 0.0041>,<at 0.25 0.0103>,<at 0.291 0.0186>,<at 0.333 0.0156>,<at 0.375 0.0208>,<at 0.416 0.0272>,<at 0.458 0.0272>,<at 0.5 0.0272>,<at 0.541 0.0272>,<at 0.583 0.0272>,<at 0.625 0.0208>,<at 0.666 0.0156>,<at 0.708 0.0192>,<at 0.75 0.0153>,<at 0.791 0.0092>,<at 0.833 0.0051>,<at 0.875 0.0062>,<at 0.916 0.0064>,<at 0.958 0.0063>,<at 1 0.0068>]}
          RedCurve {keys [<at 0 0.0033>,<at 0.041 0.0029>,<at 0.083 0.0032>,<at 0.125 0.0036>,<at 0.166 0.01>,<at 0.208 0.01>,<at 0.25 0.02>,<at 0.291 0.03>,<at 0.333 0.0086>,<at 0.375 0.0115>,<at 0.416 0.0151>,<at 0.458 0.0151>,<at 0.5 0.0151>,<at 0.541 0.0151>,<at 0.583 0.0151>,<at 0.625 0.0115>,<at 0.666 0.0086>,<at 0.708 0.03>,<at 0.75 0.03>,<at 0.791 0.02>,<at 0.833 0.01>,<at 0.875 0.01>,<at 0.916 0.0022>,<at 0.958 0.0019>,<at 1 0.0033>]}
         }}
        CurveIndex=Linked() {linkedVariable {name=NPC.FortniteNPC.TimeOfDay_0-1, type=NiagaraFloat}}
       )
       Scale Factor=&MakeFloat4FromVectorAndFloat(
        "Vector (RGB)"=&VectorFromFloat(Value=&FN_PickFloatDependOnPlatform(TargetPlatform=DataInterface() {propertyValues {Platforms {qualityLevelMask=-29}}}, TargetPlatformActive=0.2, TargetPlatformNoActive=1))
        "Float (Alpha)"=1
       )
      )
      EndColor=<rgba 0.0151 0.0272 0.0525 1>
      LerpFactor=&FN_FrontEndFloatModifier(Gameplay=0, Front End=1)
     ))) {isSetParameters=true}
   /SetVariables(Particles.HighlightDynamic:Vector4f=&MakeVector4_FromLinearColor(LinearColor=&LerpLinearColors(
      StartColor=&ColorFromCurve(
       ColorCurve=DataInterface() {propertyValues {
         AlphaCurve {keys [<at 0 1>,<at 0.125 1>,<at 0.25 1>,<at 0.375 1>,<at 0.5 1>,<at 0.625 1>,<at 0.75 1>,<at 0.875 1>,<at 1 1>]}
         BlueCurve {keys [<at 0 0.9216>,<at 0.041 0.8963>,<at 0.083 0.8308>,<at 0.125 0.7835>,<at 0.1865 0.2611>,<at 0.25 0.3013>,<at 0.333 0.3128>,<at 0.4094 0.2806>,<at 0.4889 0.2616>,<at 0.583 0.2814>,<at 0.666 0.2801>,<at 0.75 0.2471>,<at 0.791 0.2118>,<at 0.833 0.2338>,<at 0.875 0.2473>,<at 0.916 0.8388>,<at 0.958 0.9047>,<at 1 0.9216>]}
         GreenCurve {keys [<at 0 0.5616>,<at 0.041 0.5276>,<at 0.083 0.4795>,<at 0.125 0.5099>,<at 0.1865 0.2253>,<at 0.25 0.3543>,<at 0.333 0.4809>,<at 0.4094 0.4334>,<at 0.4889 0.389>,<at 0.583 0.2814>,<at 0.666 0.4785>,<at 0.75 0.2471>,<at 0.791 0.2686>,<at 0.833 0.4442>,<at 0.875 0.5188>,<at 0.916 0.5292>,<at 0.958 0.5688>,<at 1 0.6308>]}
         RedCurve {keys [<at 0 0.1584>,<at 0.041 0.1774>,<at 0.083 0.1644>,<at 0.125 0.1755>,<at 0.1865 0.6867>,<at 0.25 0.8148>,<at 0.333 0.8963>,<at 0.4094 0.9131>,<at 0.4889 0.9131>,<at 0.583 0.8714>,<at 0.666 0.8963>,<at 0.75 0.7529>,<at 0.791 0.7529>,<at 0.833 0.8469>,<at 0.875 0.8632>,<at 0.916 0.1791>,<at 0.958 0.1696>,<at 1 0.305>]}
        }}
       CurveIndex=Linked() {linkedVariable {name=NPC.FortniteNPC.TimeOfDay_0-1, type=NiagaraFloat}}
      )
      EndColor=<rgba 0.7 0.7 0.7 1>
      LerpFactor=&FN_FrontEndFloatModifier(Gameplay=0, Front End=1)
     ))) {isSetParameters=true}
  ]
  rendererClasses [&NiagaraSpriteRendererProperties]
  renderers [
   {rendererClass=&NiagaraSpriteRendererProperties}
   {rendererClass=&NiagaraSpriteRendererProperties, rendererIndex=1}
  ]
 }
 GooSmall {
  simTarget=GPUComputeSim
  emitterUpdateScript [
   &EmitterState()
   &SpawnBurst_Instantaneous(
    Spawn Count=&RandomRangeInt(Minimum=60, Maximum=80)
    Use Spawn Probability=false
    UseLoopCountLimit=false
   )
  ]
  particleSpawnScript [
   &FN_SetSpawnGroup(Number of Spawn Groups=6)
   &InitializeParticle(
    Lifetime=&RandomRangeFloat(Minimum=0.4, Maximum=0.8)
    Sprite Size Mode=::NewEnumerator1
    Uniform Sprite Size Max=5.2
    Uniform Sprite Size Min=3
    UsePositionOffset=false
    Write Material Random=true
   )
   &CameraOffset(Camera Offset Amount=2)
   /SetVariables(Particles.RandFloat:NiagaraFloat=&RandomRangeFloat(Minimum=0.2, Maximum=1.2)) {isSetParameters=true}
   /SetVariables(Particles.VisibilityTag=&MakeIntFromFloat(Float=$User.Variant)) {isSetParameters=true}
  ]
  particleUpdateScript [
   &ParticleState()
   &ScaleColor(
    Scale RGB=&MakeVectorFromLinearColorRGB(Color=&LerpLinearColors(
      StartColor=&LerpLinearColors(
       StartColor=<rgba 1 1 1 9>
       EndColor=<rgba 0.28 0.28 0.28 9>
       LerpFactor=&FN_FrontEndFloatModifier(Gameplay=0, Front End=1)
      )
      EndColor=&LerpLinearColors(
       StartColor=<rgba 0.58 0.58 0.58 9>
       EndColor=<rgba 0.58 0.58 0.58 9>
       LerpFactor=&FN_FrontEndFloatModifier(Gameplay=0, Front End=1)
      )
      LerpFactor=$User.Variant
     ))
    ScaleA=true
    ScaleRGB=true
   )
   &ScaleSpriteSize(
    Uniform Curve Scale=&Lerp_Float(A=4, B=4, Alpha=$User.Variant)
    Uniform Curve Sprite Scale=DataInterface() {propertyValues {Curve {keys [
        {arriveTangent=-0.1897, interpMode=RCIM_Cubic, leaveTangent=-0.1897, tangentMode=RCTM_User, time=0, value=0.4}
        {interpMode=RCIM_Cubic, leaveTangent=-1.3651, time=1, value=1}
       ]}}}
   ) {module=ScaleSpriteSize001}
   &ScaleSpriteSize(
    Scale Factor=&Lerp_Vector2(
     A=&Vector2DFromCurve(
      Vector2Curve=DataInterface() {propertyValues {
        XCurve {keys [{tangentMode=RCTM_User, time=0, value=1}, {tangentMode=RCTM_User, time=1, value=1}]}
        YCurve {keys [
          {interpMode=RCIM_Cubic, leaveTangentWeight=0.3333, tangentMode=RCTM_User, time=0, value=1.5848}
          {arriveTangentWeight=0.1313, interpMode=RCIM_Cubic, leaveTangentWeight=0.2021, tangentMode=RCTM_User, time=0.524, value=2.242}
          {arriveTangentWeight=0.1354, interpMode=RCIM_Cubic, leaveTangentWeight=0.0667, tangentMode=RCTM_User, time=0.9604, value=1}
          {tangentMode=RCTM_User, time=1, value=1}
         ]}
       }}
      ScaleCurve=<1 1>
     )
     B=&Vector2DFromCurve(
      Vector2Curve=DataInterface() {propertyValues {
        XCurve {keys [{tangentMode=RCTM_User, time=0, value=1}, {tangentMode=RCTM_User, time=1, value=1}]}
        YCurve {keys [
          {interpMode=RCIM_Cubic, leaveTangentWeight=0.3333, tangentMode=RCTM_User, time=0, value=1}
          {arriveTangentWeight=0.1313, interpMode=RCIM_Cubic, leaveTangentWeight=0.2021, tangentMode=RCTM_User, time=0.3938, value=1.0461}
          {arriveTangentWeight=0.1354, interpMode=RCIM_Cubic, leaveTangentWeight=0.0667, tangentMode=RCTM_User, time=0.8, value=1}
          {tangentMode=RCTM_User, time=1, value=1}
         ]}
       }}
      ScaleCurve=<1 1>
     )
     Alpha=&UniformAOrBFloat(A=0, Distribution Weight=0.8, Evaluation Type=::NewEnumerator0, Recalculate Random Each Loop=true, Override Seed=false)
    )
    Scale Sprite Size Mode=::NewEnumerator2
   )
   &DynamicMaterialParameters(
    Index 0 Param 1=1
    Index 0 Param 2=1
    Index 0 Param 3=&FloatFromCurve(FloatCurve=DataInterface() {propertyValues {Curve {keys [
         <at 0 0>
         {arriveTangent=0.44, interpMode=RCIM_Cubic, leaveTangent=0.44, time=0.4152, value=0.014}
         {arriveTangent=1.6861, interpMode=RCIM_Cubic, leaveTangent=1.6861, time=0.8228, value=0.3621}
         {arriveTangent=6.8718, time=1, value=1}
        ]}}})
    Index 0 Param 4=&RandomRangeFloat(Minimum=0, Maximum=33)
    Index 1 Param 1=&FloatFromCurve(
     FloatCurve=DataInterface() {propertyValues {Curve {keys [
         <at 0.216 1 RCIM_Cubic>
         {interpMode=RCIM_Cubic, tangentMode=RCTM_User, time=0.3505, value=0}
         {interpMode=RCIM_Cubic, tangentMode=RCTM_User, time=0.71, value=0}
         <at 0.8425 1 RCIM_Cubic>
        ]}}}
     CurveIndex=Linked() {linkedVariable {name=NPC.FortniteNPC.TimeOfDay_0-1, type=NiagaraFloat}}
    )
    Index 1 Param 2=&FloatFromCurve(
     FloatCurve=DataInterface() {propertyValues {Curve {keys [
         <at 0.216 0.4 RCIM_Cubic>
         {interpMode=RCIM_Cubic, tangentMode=RCTM_User, time=0.3505, value=0.94}
         {interpMode=RCIM_Cubic, tangentMode=RCTM_User, time=0.71, value=0.94}
         <at 0.8425 0.4 RCIM_Cubic>
        ]}}}
     CurveIndex=Linked() {linkedVariable {name=NPC.FortniteNPC.TimeOfDay_0-1, type=NiagaraFloat}}
    )
    Module.Index0Param0Write=true
    Module.Index0Param1Write=true
    Module.Index0Param2Write=true
    Module.Index0Param3Write=true
    Module.Index1Param0Write=true
    Module.Index1Param1Write=true
    Module.Index1Param2Write=true
    Module.Index1Param3Write=true
    Write Parameter Index 1=true
   )
   &SkeletalMeshLocation(
    Bone / Socket Index=&Add_Integer(A=&SpawnGroup(), B=10)
    Bone Sampling Mode=::NewEnumerator4
    MaskSpawnGroup=false
    Sampled Position Offset=&SelectVectorFromArray(Direct Array Index=&SpawnGroup(), Vector Selection Array=DataInterface() {propertyValues {FloatData [<0 0 0>]}})
    Skeletal Mesh=$System.Skeletal Mesh
   )
   /SetVariables(Particles.PosVector:Vector3f=&Add_Vector(
     A=&Multiply_Vector(
      A=&MakeVector(
       X=&RandomRangeFloat(Minimum=0.1, Maximum=1)
       Y=&Multiply_Float(A=1, B=$Particles.RandFloat)
       Z=&Multiply_Float(A=1, B=$Particles.RandFloat)
      )
      B=&SelectVectorFromArray(
       Direct Array Index=&SpawnGroup()
       Vector Selection Array=DataInterface() {propertyValues {FloatData [<37 -50 -60>, <0 -200 -100>, <-20 -40 -80>, <37 -50 60>, <0 -200 100>, <-20 -40 80>]}}
      )
     )
     B=<0 0 0>
    )) {isSetParameters=true, module=SetVariables_#2}
   /SetVariables(Particles.Position=&AddVectorToPosition(
     Position=$Particles.Position
     Vector=&Add_Vector(
      A=&VectorFromCurve(
       VectorCurve=DataInterface() {propertyValues {
         XCurve {keys [
           {arriveTangent=1.4099, interpMode=RCIM_Cubic, leaveTangent=1.4099, tangentMode=RCTM_User, time=0, value=0}
           {arriveTangent=0.8362, tangentMode=RCTM_User, time=1, value=1.3292}
          ]}
         YCurve {keys [
           {arriveTangent=0.0423, interpMode=RCIM_Cubic, leaveTangent=0.0423, tangentMode=RCTM_User, time=0, value=0}
           {arriveTangent=0.1842, interpMode=RCIM_Cubic, leaveTangent=0.1842, tangentMode=RCTM_User, time=1, value=0.2202}
          ]}
         ZCurve {keys [{arriveTangent=1.5712, interpMode=RCIM_Cubic, leaveTangent=1.5712, tangentMode=RCTM_User, time=0, value=0}, <at 1 1.4648 RCIM_Cubic>]}
        }}
       Scale Curve=$Particles.PosVector
      )
      B=<0 0 0>
     )
    )) {isSetParameters=true, module=SetVariables_#3}
   &SpriteFacingAndAlignment(
    Sprite Alignment=&AddVectorToPosition(
     Position=$Particles.Position
     Vector=&Add_Vector(
      A=&VectorFromCurve(
       VectorCurve=DataInterface() {propertyValues {
         XCurve {keys [
           {arriveTangent=1.4099, interpMode=RCIM_Cubic, leaveTangent=1.4099, tangentMode=RCTM_User, time=0, value=0}
           {arriveTangent=0.8362, tangentMode=RCTM_User, time=1, value=1.3292}
          ]}
         YCurve {keys [
           {arriveTangent=0.0423, interpMode=RCIM_Cubic, leaveTangent=0.0423, tangentMode=RCTM_User, time=0, value=0}
           {arriveTangent=0.1842, interpMode=RCIM_Cubic, leaveTangent=0.1842, tangentMode=RCTM_User, time=1, value=0.2202}
          ]}
         ZCurve {keys [{arriveTangent=1.5712, interpMode=RCIM_Cubic, leaveTangent=1.5712, tangentMode=RCTM_User, time=0, value=0}, <at 1 1.4648 RCIM_Cubic>]}
        }}
       Scale Curve=$Particles.PosVector
      )
      B=<0 0 0>
     )
    )
    WriteAlignment=true
    WriteFacing=false
   )
   /SetVariables(Particles.VariantDynamic:Vector4f=&MakeVector4(
     X=1
     Y=1
     Z=&FloatFromCurve(FloatCurve=DataInterface() {propertyValues {Curve {keys [
          <at 0 0>
          {arriveTangent=1, interpMode=RCIM_Cubic, leaveTangent=1, time=0.3628, value=0}
          {arriveTangent=3.3266, tangentMode=RCTM_User, time=1, value=1}
         ]}}})
     W=&RandomRangeFloat(Minimum=0, Maximum=1)
    )) {isSetParameters=true}
   /SetVariables(Particles.RimLightDynamic:Vector4f=&MakeVector4_FromLinearColor(LinearColor=&LerpLinearColors(
      StartColor=&MultiplyLinearColor(
       Linear Color=&ColorFromCurve(
        ColorCurve=DataInterface() {propertyValues {
          AlphaCurve {keys [<at 0 1>,<at 0.125 1>,<at 0.25 1>,<at 0.375 1>,<at 0.5 1>,<at 0.625 1>,<at 0.75 1>,<at 0.875 1>,<at 1 1>]}
          BlueCurve {keys [<at 0 0.01>,<at 0.041 0.01>,<at 0.083 0.01>,<at 0.125 0.01>,<at 0.166 0.0077>,<at 0.208 0.0045>,<at 0.25 0.0092>,<at 0.291 0.0136>,<at 0.333 0.03>,<at 0.375 0.04>,<at 0.416 0.0525>,<at 0.458 0.0525>,<at 0.5 0.0525>,<at 0.541 0.0525>,<at 0.583 0.0525>,<at 0.625 0.04>,<at 0.666 0.03>,<at 0.708 0.0152>,<at 0.75 0.0153>,<at 0.791 0.0079>,<at 0.833 0.0025>,<at 0.875 0.0031>,<at 0.916 0.01>,<at 0.958 0.01>,<at 1 0.01>]}
          GreenCurve {keys [<at 0 0.0068>,<at 0.041 0.0064>,<at 0.083 0.0064>,<at 0.125 0.0071>,<at 0.166 0.0056>,<at 0.208 0.0041>,<at 0.25 0.0103>,<at 0.291 0.0186>,<at 0.333 0.0156>,<at 0.375 0.0208>,<at 0.416 0.0272>,<at 0.458 0.0272>,<at 0.5 0.0272>,<at 0.541 0.0272>,<at 0.583 0.0272>,<at 0.625 0.0208>,<at 0.666 0.0156>,<at 0.708 0.0192>,<at 0.75 0.0153>,<at 0.791 0.0092>,<at 0.833 0.0051>,<at 0.875 0.0062>,<at 0.916 0.0064>,<at 0.958 0.0063>,<at 1 0.0068>]}
          RedCurve {keys [<at 0 0.0033>,<at 0.041 0.0029>,<at 0.083 0.0032>,<at 0.125 0.0036>,<at 0.166 0.01>,<at 0.208 0.01>,<at 0.25 0.02>,<at 0.291 0.03>,<at 0.333 0.0086>,<at 0.375 0.0115>,<at 0.416 0.0151>,<at 0.458 0.0151>,<at 0.5 0.0151>,<at 0.541 0.0151>,<at 0.583 0.0151>,<at 0.625 0.0115>,<at 0.666 0.0086>,<at 0.708 0.03>,<at 0.75 0.03>,<at 0.791 0.02>,<at 0.833 0.01>,<at 0.875 0.01>,<at 0.916 0.0022>,<at 0.958 0.0019>,<at 1 0.0033>]}
         }}
        CurveIndex=Linked() {linkedVariable {name=NPC.FortniteNPC.TimeOfDay_0-1, type=NiagaraFloat}}
       )
       Scale Factor=&MakeFloat4FromVectorAndFloat(
        "Vector (RGB)"=&VectorFromFloat(Value=&FN_PickFloatDependOnPlatform(TargetPlatform=DataInterface() {propertyValues {Platforms {qualityLevelMask=-29}}}, TargetPlatformActive=0.2, TargetPlatformNoActive=1))
        "Float (Alpha)"=1
       )
      )
      EndColor=<rgba 0.0151 0.0272 0.0525 1>
      LerpFactor=&FN_FrontEndFloatModifier(Gameplay=0, Front End=1)
     ))) {isSetParameters=true}
   /SetVariables(Particles.HighlightDynamic:Vector4f=&MakeVector4_FromLinearColor(LinearColor=&LerpLinearColors(
      StartColor=&ColorFromCurve(
       ColorCurve=DataInterface() {propertyValues {
         AlphaCurve {keys [<at 0 1>,<at 0.125 1>,<at 0.25 1>,<at 0.375 1>,<at 0.5 1>,<at 0.625 1>,<at 0.75 1>,<at 0.875 1>,<at 1 1>]}
         BlueCurve {keys [<at 0 0.9216>,<at 0.041 0.8963>,<at 0.083 0.8308>,<at 0.125 0.7835>,<at 0.1865 0.2611>,<at 0.25 0.3013>,<at 0.333 0.3128>,<at 0.4094 0.2806>,<at 0.4889 0.2616>,<at 0.583 0.2814>,<at 0.666 0.2801>,<at 0.75 0.2471>,<at 0.791 0.2118>,<at 0.833 0.2338>,<at 0.875 0.2473>,<at 0.916 0.8388>,<at 0.958 0.9047>,<at 1 0.9216>]}
         GreenCurve {keys [<at 0 0.5616>,<at 0.041 0.5276>,<at 0.083 0.4795>,<at 0.125 0.5099>,<at 0.1865 0.2253>,<at 0.25 0.3543>,<at 0.333 0.4809>,<at 0.4094 0.4334>,<at 0.4889 0.389>,<at 0.583 0.2814>,<at 0.666 0.4785>,<at 0.75 0.2471>,<at 0.791 0.2686>,<at 0.833 0.4442>,<at 0.875 0.5188>,<at 0.916 0.5292>,<at 0.958 0.5688>,<at 1 0.6308>]}
         RedCurve {keys [<at 0 0.1584>,<at 0.041 0.1774>,<at 0.083 0.1644>,<at 0.125 0.1755>,<at 0.1865 0.6867>,<at 0.25 0.8148>,<at 0.333 0.8963>,<at 0.4094 0.9131>,<at 0.4889 0.9131>,<at 0.583 0.8714>,<at 0.666 0.8963>,<at 0.75 0.7529>,<at 0.791 0.7529>,<at 0.833 0.8469>,<at 0.875 0.8632>,<at 0.916 0.1791>,<at 0.958 0.1696>,<at 1 0.305>]}
        }}
       CurveIndex=Linked() {linkedVariable {name=NPC.FortniteNPC.TimeOfDay_0-1, type=NiagaraFloat}}
      )
      EndColor=<rgba 0.7 0.7 0.7 1>
      LerpFactor=&FN_FrontEndFloatModifier(Gameplay=0, Front End=1)
     ))) {isSetParameters=true}
  ]
  rendererClasses [&NiagaraSpriteRendererProperties]
  renderers [
   {rendererClass=&NiagaraSpriteRendererProperties}
   {rendererClass=&NiagaraSpriteRendererProperties, rendererIndex=1}
  ]
 }
 GooSmallWings {
  emitterUpdateScript [
   &EmitterState()
   &SpawnRate(
    SpawnRate=&FloatFromCurve(
     FloatCurve=DataInterface() {propertyValues {Curve {keys [
         <at 0.0291 0 RCIM_Cubic>
         {arriveTangent=0.7122, interpMode=RCIM_Cubic, leaveTangent=0.7122, time=0.0534, value=0.1}
         {interpMode=RCIM_Cubic, tangentMode=RCTM_User, time=0.242, value=0.1516}
         {arriveTangent=4.9766, interpMode=RCIM_Cubic, leaveTangent=4.9766, tangentMode=RCTM_User, time=0.2816, value=0.5812}
         {arriveTangent=-2.9808, interpMode=RCIM_Cubic, leaveTangent=-2.9808, time=0.3887, value=0.5491}
         <at 0.4766 0>
        ]}}}
     CurveIndex=$Emitter.NormalizedLoopAge
     Scale Curve=80
    )
    Use Spawn Probability=false
   )
  ]
  particleSpawnScript [
   &FN_SetSpawnGroup(Number of Spawn Groups=10)
   &InitializeParticle(
    Lifetime=&RandomRangeFloat(Minimum=0.5, Maximum=1.1)
    Sprite Size Mode=::NewEnumerator1
    Uniform Sprite Size Max=5.2
    Uniform Sprite Size Min=3
    UsePositionOffset=false
    Write Material Random=true
   )
   &CameraOffset(Camera Offset Amount=2)
   &SkeletalMeshLocation(
    Bone / Socket Index=&SelectIntFromArray(
     Direct Array Index=&SpawnGroup()
     Int Selection Array=DataInterface() {propertyValues {IntData [0,1,2,3,4,5,17,18,19,20,21,22]}}
    )
    Bone Sampling Mode=::NewEnumerator4
    MaskSpawnGroup=false
    Sampled Position Offset=&RandomRangeVector(Minimum=<-1.5 -1.5 -1.5>, Maximum=<1.5 1.5 1.5>)
    Skeletal Mesh=$System.Skeletal Mesh
   )
   /SetVariables(Particles.VisibilityTag=&MakeIntFromFloat(Float=$User.Variant)) {isSetParameters=true}
   &AddVelocity(
    Rotation Coordinate Space=::World
    Velocity=&RandomRangeVector(Minimum=<-100 -100 0>, Maximum=<100 100 0>)
   )
  ]
  particleUpdateScript [
   &ParticleState()
   &ScaleColor(
    Scale RGB=&MakeVectorFromLinearColorRGB(Color=&LerpLinearColors(
      StartColor=&LerpLinearColors(
       StartColor=<rgba 1 1 1 9>
       EndColor=<rgba 0.28 0.28 0.28 9>
       LerpFactor=&FN_FrontEndFloatModifier(Gameplay=0, Front End=1)
      )
      EndColor=&LerpLinearColors(
       StartColor=<rgba 0.58 0.58 0.58 9>
       EndColor=<rgba 0.58 0.58 0.58 9>
       LerpFactor=&FN_FrontEndFloatModifier(Gameplay=0, Front End=1)
      )
      LerpFactor=$User.Variant
     ))
    ScaleA=true
    ScaleRGB=true
   )
   &ScaleSpriteSize(
    Uniform Curve Scale=&Lerp_Float(A=3, B=3, Alpha=$User.Variant)
    Uniform Curve Sprite Scale=DataInterface() {propertyValues {Curve {keys [
        {arriveTangent=1.153, interpMode=RCIM_Cubic, leaveTangent=1.153, tangentMode=RCTM_User, time=0, value=0.4}
        {interpMode=RCIM_Cubic, leaveTangent=-1.3651, time=1, value=1}
       ]}}}
   ) {module=ScaleSpriteSize001}
   &ScaleSpriteSize(
    Scale Factor=&Lerp_Vector2(
     A=&Vector2DFromCurve(
      Vector2Curve=DataInterface() {propertyValues {
        XCurve {keys [{tangentMode=RCTM_User, time=0, value=1}, {tangentMode=RCTM_User, time=1, value=1}]}
        YCurve {keys [
          {interpMode=RCIM_Cubic, leaveTangentWeight=0.3333, tangentMode=RCTM_User, time=0, value=1.5848}
          {
           arriveTangent=1.3848
           arriveTangentWeight=0.1313
           interpMode=RCIM_Cubic
           leaveTangent=1.3848
           leaveTangentWeight=0.2021
           time=0.524
           value=2.7414
          }
          {arriveTangentWeight=0.1354, interpMode=RCIM_Cubic, leaveTangentWeight=0.0667, tangentMode=RCTM_User, time=0.9604, value=2.9148}
         ]}
       }}
      ScaleCurve=<0.9 1.1>
     )
     B=&Vector2DFromCurve(
      Vector2Curve=DataInterface() {propertyValues {
        XCurve {keys [{tangentMode=RCTM_User, time=0, value=1}]}
        YCurve {keys [
          {interpMode=RCIM_Cubic, leaveTangentWeight=0.3333, tangentMode=RCTM_User, time=0, value=1}
          {
           arriveTangent=0.2781
           arriveTangentWeight=0.1313
           interpMode=RCIM_Cubic
           leaveTangent=0.2781
           leaveTangentWeight=0.2021
           time=0.3938
           value=1.1704
          }
          {arriveTangentWeight=0.1354, interpMode=RCIM_Cubic, leaveTangentWeight=0.0667, time=0.8, value=1.2224}
         ]}
       }}
      ScaleCurve=<0.9 1.1>
     )
     Alpha=&UniformAOrBFloat(A=0, Distribution Weight=0.672, Evaluation Type=::NewEnumerator0, Recalculate Random Each Loop=true, Override Seed=false)
    )
    Scale Sprite Size Mode=::NewEnumerator2
   )
   &DynamicMaterialParameters(
    Index 0 Param 1=1
    Index 0 Param 2=1
    Index 0 Param 3=&FloatFromCurve(FloatCurve=DataInterface() {propertyValues {Curve {keys [
         <at 0 0>
         {arriveTangent=0.44, interpMode=RCIM_Cubic, leaveTangent=0.44, time=0.4152, value=0.014}
         {arriveTangent=1.6861, interpMode=RCIM_Cubic, leaveTangent=1.6861, time=0.8228, value=0.3621}
         {arriveTangent=6.8718, time=1, value=1}
        ]}}})
    Index 0 Param 4=&RandomRangeFloat(Minimum=0, Maximum=33)
    Index 1 Param 1=&FloatFromCurve(
     FloatCurve=DataInterface() {propertyValues {Curve {keys [
         <at 0.216 1 RCIM_Cubic>
         {interpMode=RCIM_Cubic, tangentMode=RCTM_User, time=0.3505, value=0}
         {interpMode=RCIM_Cubic, tangentMode=RCTM_User, time=0.71, value=0}
         <at 0.8425 1 RCIM_Cubic>
        ]}}}
     CurveIndex=Linked() {linkedVariable {name=NPC.FortniteNPC.TimeOfDay_0-1, type=NiagaraFloat}}
    )
    Index 1 Param 2=&FloatFromCurve(
     FloatCurve=DataInterface() {propertyValues {Curve {keys [
         <at 0.216 0.4 RCIM_Cubic>
         {interpMode=RCIM_Cubic, tangentMode=RCTM_User, time=0.3505, value=0.94}
         {interpMode=RCIM_Cubic, tangentMode=RCTM_User, time=0.71, value=0.94}
         <at 0.8425 0.4 RCIM_Cubic>
        ]}}}
     CurveIndex=Linked() {linkedVariable {name=NPC.FortniteNPC.TimeOfDay_0-1, type=NiagaraFloat}}
    )
    Module.Index0Param0Write=true
    Module.Index0Param1Write=true
    Module.Index0Param2Write=true
    Module.Index0Param3Write=true
    Module.Index1Param0Write=true
    Module.Index1Param1Write=true
    Module.Index1Param2Write=true
    Module.Index1Param3Write=true
    Write Parameter Index 1=true
   )
   /SetVariables(Particles.VariantDynamic:Vector4f=&MakeVector4(
     X=1
     Y=1
     Z=&FloatFromCurve(FloatCurve=DataInterface() {propertyValues {Curve {keys [
          <at 0 0>
          {arriveTangent=1, interpMode=RCIM_Cubic, leaveTangent=1, time=0.3628, value=0}
          {arriveTangent=3.3266, tangentMode=RCTM_User, time=1, value=1}
         ]}}})
     W=&RandomRangeFloat(Minimum=0, Maximum=1)
    )) {isSetParameters=true}
   &GravityForce(Gravity=&RandomRangeVector(Minimum=<0 0 -500>, Maximum=<0 0 -200>))
   &InheritSourceMovement(ApplyForce=false, ApplyPosition=true, ApplyPositionFalloff=false)
   &SolveForcesAndVelocity(Clamp Velocity=false, Limit Acceleration=false)
   /SetVariables(Particles.RimLightDynamic:Vector4f=&MakeVector4_FromLinearColor(LinearColor=&LerpLinearColors(
      StartColor=&MultiplyLinearColor(
       Linear Color=&ColorFromCurve(
        ColorCurve=DataInterface() {propertyValues {
          AlphaCurve {keys [<at 0 1>,<at 0.125 1>,<at 0.25 1>,<at 0.375 1>,<at 0.5 1>,<at 0.625 1>,<at 0.75 1>,<at 0.875 1>,<at 1 1>]}
          BlueCurve {keys [<at 0 0.01>,<at 0.041 0.01>,<at 0.083 0.01>,<at 0.125 0.01>,<at 0.166 0.0077>,<at 0.208 0.0045>,<at 0.25 0.0092>,<at 0.291 0.0136>,<at 0.333 0.03>,<at 0.375 0.04>,<at 0.416 0.0525>,<at 0.458 0.0525>,<at 0.5 0.0525>,<at 0.541 0.0525>,<at 0.583 0.0525>,<at 0.625 0.04>,<at 0.666 0.03>,<at 0.708 0.0152>,<at 0.75 0.0153>,<at 0.791 0.0079>,<at 0.833 0.0025>,<at 0.875 0.0031>,<at 0.916 0.01>,<at 0.958 0.01>,<at 1 0.01>]}
          GreenCurve {keys [<at 0 0.0068>,<at 0.041 0.0064>,<at 0.083 0.0064>,<at 0.125 0.0071>,<at 0.166 0.0056>,<at 0.208 0.0041>,<at 0.25 0.0103>,<at 0.291 0.0186>,<at 0.333 0.0156>,<at 0.375 0.0208>,<at 0.416 0.0272>,<at 0.458 0.0272>,<at 0.5 0.0272>,<at 0.541 0.0272>,<at 0.583 0.0272>,<at 0.625 0.0208>,<at 0.666 0.0156>,<at 0.708 0.0192>,<at 0.75 0.0153>,<at 0.791 0.0092>,<at 0.833 0.0051>,<at 0.875 0.0062>,<at 0.916 0.0064>,<at 0.958 0.0063>,<at 1 0.0068>]}
          RedCurve {keys [<at 0 0.0033>,<at 0.041 0.0029>,<at 0.083 0.0032>,<at 0.125 0.0036>,<at 0.166 0.01>,<at 0.208 0.01>,<at 0.25 0.02>,<at 0.291 0.03>,<at 0.333 0.0086>,<at 0.375 0.0115>,<at 0.416 0.0151>,<at 0.458 0.0151>,<at 0.5 0.0151>,<at 0.541 0.0151>,<at 0.583 0.0151>,<at 0.625 0.0115>,<at 0.666 0.0086>,<at 0.708 0.03>,<at 0.75 0.03>,<at 0.791 0.02>,<at 0.833 0.01>,<at 0.875 0.01>,<at 0.916 0.0022>,<at 0.958 0.0019>,<at 1 0.0033>]}
         }}
        CurveIndex=Linked() {linkedVariable {name=NPC.FortniteNPC.TimeOfDay_0-1, type=NiagaraFloat}}
       )
       Scale Factor=&MakeFloat4FromVectorAndFloat(
        "Vector (RGB)"=&VectorFromFloat(Value=&FN_PickFloatDependOnPlatform(TargetPlatform=DataInterface() {propertyValues {Platforms {qualityLevelMask=-29}}}, TargetPlatformActive=0.2, TargetPlatformNoActive=1))
        "Float (Alpha)"=1
       )
      )
      EndColor=<rgba 0.0151 0.0272 0.0525 1>
      LerpFactor=&FN_FrontEndFloatModifier(Gameplay=0, Front End=1)
     ))) {isSetParameters=true}
   /SetVariables(Particles.HighlightDynamic:Vector4f=&MakeVector4_FromLinearColor(LinearColor=&LerpLinearColors(
      StartColor=&ColorFromCurve(
       ColorCurve=DataInterface() {propertyValues {
         AlphaCurve {keys [<at 0 1>,<at 0.125 1>,<at 0.25 1>,<at 0.375 1>,<at 0.5 1>,<at 0.625 1>,<at 0.75 1>,<at 0.875 1>,<at 1 1>]}
         BlueCurve {keys [<at 0 0.9216>,<at 0.041 0.8963>,<at 0.083 0.8308>,<at 0.125 0.7835>,<at 0.1865 0.2611>,<at 0.25 0.3013>,<at 0.333 0.3128>,<at 0.4094 0.2806>,<at 0.4889 0.2616>,<at 0.583 0.2814>,<at 0.666 0.2801>,<at 0.75 0.2471>,<at 0.791 0.2118>,<at 0.833 0.2338>,<at 0.875 0.2473>,<at 0.916 0.8388>,<at 0.958 0.9047>,<at 1 0.9216>]}
         GreenCurve {keys [<at 0 0.5616>,<at 0.041 0.5276>,<at 0.083 0.4795>,<at 0.125 0.5099>,<at 0.1865 0.2253>,<at 0.25 0.3543>,<at 0.333 0.4809>,<at 0.4094 0.4334>,<at 0.4889 0.389>,<at 0.583 0.2814>,<at 0.666 0.4785>,<at 0.75 0.2471>,<at 0.791 0.2686>,<at 0.833 0.4442>,<at 0.875 0.5188>,<at 0.916 0.5292>,<at 0.958 0.5688>,<at 1 0.6308>]}
         RedCurve {keys [<at 0 0.1584>,<at 0.041 0.1774>,<at 0.083 0.1644>,<at 0.125 0.1755>,<at 0.1865 0.6867>,<at 0.25 0.8148>,<at 0.333 0.8963>,<at 0.4094 0.9131>,<at 0.4889 0.9131>,<at 0.583 0.8714>,<at 0.666 0.8963>,<at 0.75 0.7529>,<at 0.791 0.7529>,<at 0.833 0.8469>,<at 0.875 0.8632>,<at 0.916 0.1791>,<at 0.958 0.1696>,<at 1 0.305>]}
        }}
       CurveIndex=Linked() {linkedVariable {name=NPC.FortniteNPC.TimeOfDay_0-1, type=NiagaraFloat}}
      )
      EndColor=<rgba 0.7 0.7 0.7 1>
      LerpFactor=&FN_FrontEndFloatModifier(Gameplay=0, Front End=1)
     ))) {isSetParameters=true}
  ]
  rendererClasses [&NiagaraSpriteRendererProperties]
  renderers [
   {rendererClass=&NiagaraSpriteRendererProperties}
   {rendererClass=&NiagaraSpriteRendererProperties, rendererIndex=1}
  ]
 }
}
summary {
 systemName=NS_Glider_MissusMind_Deploy
 emitters [
  {emitterName=GooBig, rendererClasses [&NiagaraSpriteRendererProperties]}
  {emitterName=GooSmall, simTarget=GPUComputeSim, rendererClasses [&NiagaraSpriteRendererProperties]}
  {emitterName=GooSmallWings, rendererClasses [&NiagaraSpriteRendererProperties]}
  {emitterName=BackpackSplash, rendererClasses [&NiagaraSpriteRendererProperties]}
 ]
 userVariables [
  User.bIsFrontEnd {type=NiagaraBool}
  User.bIsFrontEndPreview {type=NiagaraBool}
  User.bIsFullyDeployed {type=NiagaraBool}
  User.ForwardDot {type=NiagaraFloat, defaultValue=0, description=""}
  User.RibbonIDOffset {type=NiagaraInt32}
  User.RightDot {type=NiagaraFloat, defaultValue=0, description=""}
  User.Variant {type=NiagaraFloat, defaultValue=0, description=""}
 ]
}
