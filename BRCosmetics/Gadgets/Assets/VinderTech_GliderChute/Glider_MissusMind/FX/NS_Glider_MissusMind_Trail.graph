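# Niagara trail effect for the MissusMind glider, pairing dripping goo sprites with streaking speed lines.
# Goo spawns on skeletal mesh sockets by spawn group and stretches downward and against owner velocity, with User.Variant setting tint and visibility tag.
# Time-of-day curves feed goo rim and highlight colors and scale speed-line opacity in gameplay.
# System Timeline ramps trail strength, while front-end modifiers swap in calmer, fixed-color menu values.
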NiagaraSystem /BRCosmetics/Gadgets/Assets/VinderTech_GliderChute/Glider_MissusMind/FX/NS_Glider_MissusMind_Trail
dependencies {
 usedDataInterfaces [
  &NiagaraDataInterfaceCurve
  &NiagaraDataInterfaceVector2DCurve
  &NiagaraDataInterfaceArrayFloat
  &NiagaraDataInterfaceArrayFloat3
  &NiagaraDataInterfaceVectorCurve
  &NiagaraDataInterfaceColorCurve
  &NiagaraDataInterfaceSkeletalMesh
 ]
 usedDynamicInputs [
  &MaskVector2dBySpawnGroup
  &RandomRangeVector2D
  &MakeFloatFromVector2d
  &SetBoolByFloatComparison
  &RemapRange
  &TransformVector
 ]
 usedModules [
  &self:SystemSpawnScript.SystemScriptSource.SystemScriptGraph.NiagaraNodeAssignment_0.SetVariables_#1
  &SystemState
  &Timeline
  &FN_Integrator
  &InterpolateOverTime
  &self:SystemSpawnScript.SystemScriptSource.SystemScriptGraph.NiagaraNodeAssignment_1.SetVariables_#2
 ]
}
emitters {
 Goo {
  emitterUpdateScript [
   &EmitterState()
   &SpawnRate(
    Spawn Probability=0.5
    SpawnRate=&Multiply_Float(
     A=&RandomRangeFloat(Minimum=12, Maximum=24)
     B=&FloatFromCurve(FloatCurve=DataInterface() {propertyValues {Curve {keys [<at 0 0.44>,<at 1 0.8>]}}}, CurveIndex=$System.Timeline.Time)
    )
    Use Spawn Probability=true
   ) {module=SpawnRate002}
  ]
  particleSpawnScript [
   &FN_SetSpawnGroup(Number of Spawn Groups=10)
   &InitializeParticle(
    Lifetime=&Multiply_Float(
     A=&RandomRangeFloat(Minimum=0.6, Maximum=0.9)
     B=&FloatFromCurve(FloatCurve=DataInterface() {propertyValues {Curve {keys [<at 0.1 1>,<at 0.5 0.88>]}}}, CurveIndex=$System.Timeline.Time)
    )
    Sprite Size Mode=::NewEnumerator1
    Uniform Sprite Size Max=4
    Uniform Sprite Size Min=3
    UsePositionOffset=false
    Write Material Random=true
   )
   &CameraOffset(Camera Offset Amount=2)
   /SetVariables(Particles.VisibilityTag=&MakeIntFromFloat(Float=$User.Variant)) {isSetParameters=true}
  ]
  particleUpdateScript [
   &ParticleState()
   &ScaleColor(
    Scale RGB=&MakeVectorFromLinearColorRGB(Color=&LerpLinearColors(
      StartColor=&LerpLinearColors(
       StartColor=<rgba 1 1 1 9>
       EndColor=<rgba 0.28 0.28 0.28 9>
       LerpFactor=&FN_FrontEndFloatModifier(Gameplay=0, Front End=1)
      )
      EndColor=&LerpLinearColors(
       StartColor=<rgba 0.58 0.58 0.58 9>
       EndColor=<rgba 0.58 0.58 0.58 9>
       LerpFactor=&FN_FrontEndFloatModifier(Gameplay=0, Front End=1)
      )
      LerpFactor=$User.Variant
     ))
    ScaleA=true
    ScaleRGB=true
   )
   &ScaleSpriteSize(
    Uniform Curve Scale=&Multiply_Float(
     A=&Lerp_Float(A=3.4, B=3.8, Alpha=$System.Timeline.Time)
     B=&Lerp_Float(A=1, B=1, Alpha=$User.Variant)
    )
    Uniform Curve Sprite Scale=DataInterface() {propertyValues {Curve {keys [
        <at 0 0 RCIM_Cubic>
        <at 0.301 1>
        {arriveTangent=-1.4306, interpMode=RCIM_Cubic, leaveTangent=-1.4306, time=0.6, value=1}
        {tangentMode=RCTM_User, time=1, value=0}
       ]}}}
   ) {module=ScaleSpriteSize001}
   &ScaleSpriteSize(
    Scale Factor=&Lerp_Vector2(
     A=&Vector2DFromCurve(
      Vector2Curve=DataInterface() {propertyValues {
        XCurve {keys [{tangentMode=RCTM_User, time=0, value=1}, {tangentMode=RCTM_User, time=1, value=1}]}
        YCurve {keys [
          {interpMode=RCIM_Cubic, leaveTangentWeight=0.3333, tangentMode=RCTM_User, time=0, value=1.7137}
          {arriveTangentWeight=0.1313, interpMode=RCIM_Cubic, leaveTangentWeight=0.2021, tangentMode=RCTM_User, time=0.524, value=2.683}
          {arriveTangentWeight=0.1354, interpMode=RCIM_Cubic, leaveTangentWeight=0.0667, tangentMode=RCTM_User, time=0.9604, value=1}
          {tangentMode=RCTM_User, time=1, value=1}
         ]}
       }}
      ScaleCurve=&Lerp_Vector2(A=<0.8 1.3>, B=<0.8 1.3>, Alpha=$User.Variant)
     )
     B=&Vector2DFromCurve(
      Vector2Curve=DataInterface() {propertyValues {
        XCurve {keys [{tangentMode=RCTM_User, time=0, value=1}, {tangentMode=RCTM_User, time=1, value=1}]}
        YCurve {keys [
          {interpMode=RCIM_Cubic, leaveTangentWeight=0.3333, tangentMode=RCTM_User, time=0, value=1}
          {arriveTangentWeight=0.1313, interpMode=RCIM_Cubic, leaveTangentWeight=0.2021, tangentMode=RCTM_User, time=0.3938, value=1}
          {arriveTangentWeight=0.1354, interpMode=RCIM_Cubic, leaveTangentWeight=0.0667, tangentMode=RCTM_User, time=0.8, value=1}
          {tangentMode=RCTM_User, time=1, value=1}
         ]}
       }}
      ScaleCurve=<0.8 1.3>
     )
     Alpha=&RandomRangeFloat(Minimum=0, Maximum=1)
    )
    Scale Sprite Size Mode=::NewEnumerator2
   )
   &ScaleSpriteSize(
    Scale Sprite Size Mode=::NewEnumerator0
    Uniform Scale Factor=&Lerp_Float(
     A=1
     B=&UniformAOrBFloat(A=0, Distribution Weight=0.4, Evaluation Type=::NewEnumerator0, Recalculate Random Each Loop=true, Override Seed=false)
     Alpha=&SelectFloatFromArray(
      Direct Array Index=&SpawnGroup()
      Float Selection Array=DataInterface() {propertyValues {FloatData [0,0,0,0,0,1,1]}}
     )
    )
   ) {module=ScaleSpriteSize002}
   &DynamicMaterialParameters(
    Index 0 Param 1=1
    Index 0 Param 2=1
    Index 0 Param 3=&FloatFromCurve(FloatCurve=DataInterface() {propertyValues {Curve {keys [
         <at 0 0>
         {arriveTangent=0.5283, interpMode=RCIM_Cubic, leaveTangent=0.5283, time=0.3079, value=0.014}
         {arriveTangent=1.4247, interpMode=RCIM_Cubic, leaveTangent=1.4247, time=0.6853, value=0.3621}
         {arriveTangent=3.0211, tangentMode=RCTM_User, time=1, value=1}
        ]}}})
    Index 0 Param 4=&RandomRangeFloat(Minimum=0, Maximum=33)
    Index 1 Param 1=&FloatFromCurve(
     FloatCurve=DataInterface() {propertyValues {Curve {keys [
         <at 0.216 1 RCIM_Cubic>
         {interpMode=RCIM_Cubic, tangentMode=RCTM_User, time=0.3505, value=0}
         {interpMode=RCIM_Cubic, tangentMode=RCTM_User, time=0.71, value=0}
         <at 0.8425 1 RCIM_Cubic>
        ]}}}
     CurveIndex=Linked() {linkedVariable {name=NPC.FortniteNPC.TimeOfDay_0-1, type=NiagaraFloat}}
    )
    Index 1 Param 2=&FloatFromCurve(
     FloatCurve=DataInterface() {propertyValues {Curve {keys [
         <at 0.216 0.4 RCIM_Cubic>
         {interpMode=RCIM_Cubic, tangentMode=RCTM_User, time=0.3505, value=0.94}
         {interpMode=RCIM_Cubic, tangentMode=RCTM_User, time=0.71, value=0.94}
         <at 0.8425 0.4 RCIM_Cubic>
        ]}}}
     CurveIndex=Linked() {linkedVariable {name=NPC.FortniteNPC.TimeOfDay_0-1, type=NiagaraFloat}}
    )
    Module.Index0Param0Write=true
    Module.Index0Param1Write=true
    Module.Index0Param2Write=true
    Module.Index0Param3Write=true
    Module.Index1Param0Write=true
    Module.Index1Param1Write=true
    Module.Index1Param2Write=true
    Module.Index1Param3Write=true
    Write Parameter Index 1=true
   )
   &SkeletalMeshLocation(
    Bone / Socket Index=&SpawnGroup()
    Bone Sampling Mode=::NewEnumerator4
    MaskSpawnGroup=false
    Sampled Position Offset=&SelectVectorFromArray(
     Direct Array Index=&SpawnGroup()
     Vector Selection Array=DataInterface() {propertyValues {FloatData [<-1 0 0>]}}
    )
    Skeletal Mesh=$System.Skeletal Mesh
   )
   /SetVariables(Particles.Position=&AddVectorToPosition(
     Position=$Particles.Position
     Vector=&VectorFromCurve(
      VectorCurve=DataInterface() {propertyValues {
        XCurve {keys [
          {arriveTangent=0.6026, interpMode=RCIM_Cubic, leaveTangent=0.6026, tangentMode=RCTM_User, time=0, value=0}
          {arriveTangent=1.7451, tangentMode=RCTM_User, time=1, value=1}
         ]}
        YCurve {keys [
          {arriveTangent=0.6026, interpMode=RCIM_Cubic, leaveTangent=0.6026, tangentMode=RCTM_User, time=0, value=0}
          {arriveTangent=1.7451, tangentMode=RCTM_User, time=1, value=1}
         ]}
        ZCurve {keys [
          {arriveTangent=0.6026, interpMode=RCIM_Cubic, leaveTangent=0.6026, tangentMode=RCTM_User, time=0, value=0}
          {arriveTangent=1.7451, tangentMode=RCTM_User, time=1, value=1}
         ]}
       }}
      Scale Curve=&Add_Vector(
       A=&RandomRangeVector(
        Minimum=&Lerp_Vector(A=<0 0 -10>, B=<0 0 -10>, Alpha=$System.Timeline.Time)
        Maximum=&Lerp_Vector(A=<0 0 -60>, B=<0 0 -70>, Alpha=$System.Timeline.Time)
       )
       B=&Lerp_Vector(
        A=&Multiply_Vector(
         A=&NormalizeVector(Vector To Normalize=$Engine.Owner.Velocity, Safe Normalize=true, Safe Fallback Threshold=0, Safe Fallback Vector=<0 0 0.001>)
         B=&Lerp_Vector(A=<-24 -24 0>, B=<-160 -160 0>, Alpha=$System.Timeline.Time)
        )
        B=<0 0 0>
        Alpha=&FN_FrontEndFloatModifier(Gameplay=0, Front End=1)
       )
      )
     )
    )) {isSetParameters=true}
   &SpriteFacingAndAlignment(
    Sprite Alignment=&Add_Vector(
     A=&RandomRangeVector(
      Minimum=&Lerp_Vector(A=<0 0 -10>, B=<0 0 -10>, Alpha=$System.Timeline.Time)
      Maximum=&Lerp_Vector(A=<0 0 -60>, B=<0 0 -70>, Alpha=$System.Timeline.Time)
     )
     B=&Lerp_Vector(
      A=&Multiply_Vector(
       A=&NormalizeVector(Vector To Normalize=$Engine.Owner.Velocity, Safe Normalize=true, Safe Fallback Threshold=0, Safe Fallback Vector=<0 0 0.001>)
       B=&Lerp_Vector(A=<-24 -24 0>, B=<-160 -160 0>, Alpha=$System.Timeline.Time)
      )
      B=<0 0 0>
      Alpha=&FN_FrontEndFloatModifier(Gameplay=0, Front End=1)
     )
    )
    WriteAlignment=true
    WriteFacing=false
   )
   /SetVariables(Particles.VariantDynamic:Vector4f=&MakeVector4(
     X=1
     Y=1
     Z=&FloatFromCurve(FloatCurve=DataInterface() {propertyValues {Curve {keys [
          <at 0 0>
          {arriveTangent=1, interpMode=RCIM_Cubic, leaveTangent=1, time=0.3628, value=0}
          {arriveTangent=3.3266, tangentMode=RCTM_User, time=1, value=1}
         ]}}})
     W=&RandomRangeFloat(Minimum=0, Maximum=1)
    )) {isSetParameters=true}
   /SetVariables(Particles.RimLightDynamic:Vector4f=&MakeVector4_FromLinearColor(LinearColor=&LerpLinearColors(
      StartColor=&ColorFromCurve(
       ColorCurve=DataInterface() {propertyValues {
         AlphaCurve {keys [<at 0 1>,<at 0.125 1>,<at 0.25 1>,<at 0.375 1>,<at 0.5 1>,<at 0.625 1>,<at 0.75 1>,<at 0.875 1>,<at 1 1>]}
         BlueCurve {keys [<at 0 0.01>,<at 0.041 0.01>,<at 0.083 0.01>,<at 0.125 0.01>,<at 0.166 0.0077>,<at 0.208 0.0045>,<at 0.25 0.0092>,<at 0.291 0.0136>,<at 0.333 0.03>,<at 0.375 0.04>,<at 0.416 0.0525>,<at 0.458 0.0525>,<at 0.5 0.0525>,<at 0.541 0.0525>,<at 0.583 0.0525>,<at 0.625 0.04>,<at 0.666 0.03>,<at 0.708 0.0152>,<at 0.75 0.0153>,<at 0.791 0.0079>,<at 0.833 0.0025>,<at 0.875 0.0031>,<at 0.916 0.01>,<at 0.958 0.01>,<at 1 0.01>]}
         GreenCurve {keys [<at 0 0.0068>,<at 0.041 0.0064>,<at 0.083 0.0064>,<at 0.125 0.0071>,<at 0.166 0.0056>,<at 0.208 0.0041>,<at 0.25 0.0103>,<at 0.291 0.0186>,<at 0.333 0.0156>,<at 0.375 0.0208>,<at 0.416 0.0272>,<at 0.458 0.0272>,<at 0.5 0.0272>,<at 0.541 0.0272>,<at 0.583 0.0272>,<at 0.625 0.0208>,<at 0.666 0.0156>,<at 0.708 0.0192>,<at 0.75 0.0153>,<at 0.791 0.0092>,<at 0.833 0.0051>,<at 0.875 0.0062>,<at 0.916 0.0064>,<at 0.958 0.0063>,<at 1 0.0068>]}
         RedCurve {keys [<at 0 0.0033>,<at 0.041 0.0029>,<at 0.083 0.0032>,<at 0.125 0.0036>,<at 0.166 0.01>,<at 0.208 0.01>,<at 0.25 0.02>,<at 0.291 0.03>,<at 0.333 0.0086>,<at 0.375 0.0115>,<at 0.416 0.0151>,<at 0.458 0.0151>,<at 0.5 0.0151>,<at 0.541 0.0151>,<at 0.583 0.0151>,<at 0.625 0.0115>,<at 0.666 0.0086>,<at 0.708 0.03>,<at 0.75 0.03>,<at 0.791 0.02>,<at 0.833 0.01>,<at 0.875 0.01>,<at 0.916 0.0022>,<at 0.958 0.0019>,<at 1 0.0033>]}
        }}
       CurveIndex=Linked() {linkedVariable {name=NPC.FortniteNPC.TimeOfDay_0-1, type=NiagaraFloat}}
      )
      EndColor=<rgba 0.0151 0.0272 0.0525 1>
      LerpFactor=&FN_FrontEndFloatModifier(Gameplay=0, Front End=1)
     ))) {isSetParameters=true}
   /SetVariables(Particles.HighlightDynamic:Vector4f=&MakeVector4_FromLinearColor(LinearColor=&LerpLinearColors(
      StartColor=&ColorFromCurve(
       ColorCurve=DataInterface() {propertyValues {
         AlphaCurve {keys [<at 0 1>,<at 0.125 1>,<at 0.25 1>,<at 0.375 1>,<at 0.5 1>,<at 0.625 1>,<at 0.75 1>,<at 0.875 1>,<at 1 1>]}
         BlueCurve {keys [<at 0 0.9216>,<at 0.041 0.8963>,<at 0.083 0.8308>,<at 0.125 0.7835>,<at 0.1865 0.2611>,<at 0.25 0.3013>,<at 0.333 0.3128>,<at 0.4094 0.2806>,<at 0.4889 0.2616>,<at 0.583 0.2814>,<at 0.666 0.2801>,<at 0.75 0.2471>,<at 0.791 0.2118>,<at 0.833 0.2338>,<at 0.875 0.2473>,<at 0.916 0.8388>,<at 0.958 0.9047>,<at 1 0.9216>]}
         GreenCurve {keys [<at 0 0.5616>,<at 0.041 0.5276>,<at 0.083 0.4795>,<at 0.125 0.5099>,<at 0.1865 0.2253>,<at 0.25 0.3543>,<at 0.333 0.4809>,<at 0.4094 0.4334>,<at 0.4889 0.389>,<at 0.583 0.2814>,<at 0.666 0.4785>,<at 0.75 0.2471>,<at 0.791 0.2686>,<at 0.833 0.4442>,<at 0.875 0.5188>,<at 0.916 0.5292>,<at 0.958 0.5688>,<at 1 0.6308>]}
         RedCurve {keys [<at 0 0.1584>,<at 0.041 0.1774>,<at 0.083 0.1644>,<at 0.125 0.1755>,<at 0.1865 0.6867>,<at 0.25 0.8148>,<at 0.333 0.8963>,<at 0.4094 0.9131>,<at 0.4889 0.9131>,<at 0.583 0.8714>,<at 0.666 0.8963>,<at 0.75 0.7529>,<at 0.791 0.7529>,<at 0.833 0.8469>,<at 0.875 0.8632>,<at 0.916 0.1791>,<at 0.958 0.1696>,<at 1 0.305>]}
        }}
       CurveIndex=Linked() {linkedVariable {name=NPC.FortniteNPC.TimeOfDay_0-1, type=NiagaraFloat}}
      )
      EndColor=<rgba 0.7 0.7 0.7 1>
      LerpFactor=&FN_FrontEndFloatModifier(Gameplay=0, Front End=1)
     ))) {isSetParameters=true}
  ]
  rendererClasses [&NiagaraSpriteRendererProperties]
  renderers [
   {rendererClass=&NiagaraSpriteRendererProperties}
   {rendererClass=&NiagaraSpriteRendererProperties, rendererIndex=1}
  ]
 }
 NE_Speed_Lines {
  emitterUpdateScript [
   &EmitterState(
    Delay First Loop Only=true
    Inactive Response=::NewEnumerator1
    Life Cycle Mode=::NewEnumerator1
    Loop Delay=2
    Loop Duration=1
    UseLoopDelay=true
   )
   &SpawnRate(SpawnRate=&FN_FrontEndFloatModifier(Gameplay=100, Front End=50), Use Spawn Probability=false) {module=SpawnRate001}
  ]
  particleSpawnScript [
   &InitializeParticle(
    Color=&MakeLinearColorFromVectorAndFloat(
     "Vector (RGB)"=&MakeVectorFromLinearColorRGB(Color=<rgba 1 1 1 1>)
     "Float (Alpha)"=&FN_FrontEndFloatModifier(Gameplay=0.25, Front End=0.02)
    )
    Lifetime Max=0.16
    Lifetime Min=0.1
    Lifetime Mode=::NewEnumerator1
    Sprite Size Max=<4 10>
    Sprite Size Min=<3 5>
    Sprite Size Mode=::NewEnumerator2
    UsePositionOffset=false
    Write Material Random=true
   )
   &ShapeLocation(
    Handle Radius=5
    Large Radius=&FN_FrontEndFloatModifier(Gameplay=250, Front End=111)
    MaskSpawnGroup=false
    Offset=&FN_FrontEndVector3Modifier(Gameplay=<0 -200 15>, Front End=<-111 -200 0>)
    Offset Mode=::NewEnumerator0
    Override Seed=false
    Rotation Axis=<0 1 0>
    Rotation Mode=::NewEnumerator1
    Shape Primitive=::NewEnumerator3
    UseOwnerScale=false
   )
   /SetVariables(Particles.Position=&Multiply_Vector(A=Linked() {linkedVariable {name=Particles.Position, type=NiagaraPosition}}, B=<1.2 -1 0.8>)) {isSetParameters=true}
   &AddVelocityFromPoint(
    Module.Use Falloff Distance=false
    Origin Offset=<0 1000 -452.8217>
    Velocity Strength=&FN_FrontEndFloatModifier(Gameplay=1000, Front End=0)
   )
   &AddVelocity(
    Velocity=&FN_FrontEndVector3Modifier(
     Gameplay=&Add_Vector(
      A=<0 300 0>
      B=&Multiply_VectorByFloat(
       Vector=<0 0 222>
       Float=&FloatFromCurve(FloatCurve=DataInterface() {propertyValues {Curve {keys [<at 0 0>,<at 0.6 1>]}}}, CurveIndex=$System.Timeline.Time)
      )
     )
     Front End=<44 -100 0>
    )
    Velocity Speed Scale=&FN_FrontEndFloatModifier(Gameplay=-1, Front End=1)
   )
   &CameraOffset(Camera Offset Amount=&FN_FrontEndFloatModifier(Gameplay=0, Front End=-111))
  ]
  particleUpdateScript [
   &ParticleState()
   &ScaleSpriteSize(Scale Factor=&Vector2DFromCurve(
     Vector2Curve=DataInterface() {propertyValues {XCurve {keys [<at 0 0.5 RCIM_Cubic>,<at 0.5 1 RCIM_Cubic>]}, YCurve {keys [<at 0 0 RCIM_Cubic>,<at 0.5 1 RCIM_Cubic>]}}}
     ScaleCurve=&Vector2DFromFloat(Value=1)
    ))
   &SolveForcesAndVelocity(Clamp Velocity=false, Limit Acceleration=false)
   &ScaleSpriteSizeBySpeed(Max Scale Factor=<1 50>, Min Scale Factor=<0.01 0.05>, Velocity Threshold=1)
   &ScaleColor(
    Scale Alpha=&FloatFromCurve(
     FloatCurve=DataInterface() {propertyValues {Curve {keys [
         {arriveTangent=6.8397, interpMode=RCIM_Cubic, leaveTangent=6.8397, tangentMode=RCTM_User, time=0, value=0}
         <at 0.3 1 RCIM_Cubic>
         {arriveTangent=-2.221, interpMode=RCIM_Cubic, leaveTangent=-2.221, tangentMode=RCTM_User, time=1, value=0}
        ]}}}
     Scale Curve=&FN_FrontEndFloatModifier(
      Gameplay=&Lerp_Float(
       A=0
       B=&FNTimeOfDayScalar(Night Scale=0.2, Morning Scale=0.3, Day Scale=0.4, Evening Scale=0.3)
       Alpha=&Multiply_Float(A=$System.SystemFadeIn, B=$System.Timeline.Time)
      )
      Front End=&Lerp_Float(A=0, B=0.4, Alpha=$System.SystemFadeIn)
     )
    )
    Scale RGB=&VectorFromFloat(Value=&FN_FrontEndFloatModifier(Gameplay=1, Front End=1))
    ScaleA=true
    ScaleRGB=true
   )
  ]
  rendererClasses [&NiagaraSpriteRendererProperties]
  renderers [{rendererClass=&NiagaraSpriteRendererProperties}]
 }
}
summary {
 systemName=NS_Glider_MissusMind_Trail
 emitters [
  {emitterName=Goo, rendererClasses [&NiagaraSpriteRendererProperties]}
  {emitterName=NE_Speed_Lines, rendererClasses [&NiagaraSpriteRendererProperties]}
 ]
 userVariables [
  User.bIsFrontEnd {type=NiagaraBool}
  User.bIsFrontEndPreview {type=NiagaraBool}
  User.bIsFullyDeployed {type=NiagaraBool}
  User.ForwardDot {type=NiagaraFloat, defaultValue=0, description=""}
  User.RibbonIDOffset {type=NiagaraInt32}
  User.RightDot {type=NiagaraFloat, defaultValue=0, description=""}
  User.Variant {type=NiagaraFloat, defaultValue=0, description=""}
 ]
}
