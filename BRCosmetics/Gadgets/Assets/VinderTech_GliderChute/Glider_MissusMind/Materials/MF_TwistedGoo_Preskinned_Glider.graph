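# Material function that layers animated twisted goo over a Missus Mind glider's input material by Alpha, behind a UseTwistedGoo switch.
# Panning triplanar textures projected from pre-skinned position and normal drive a two-color base and surface normals, keeping the pattern attached to the mesh.
# A UV border mask flattens seam normals, and a tight view-dot highlight supplies emissive.
# A vertex offset is computed, but output World Position Offset is constant zero.
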MaterialFunction /BRCosmetics/Gadgets/Assets/VinderTech_GliderChute/Glider_MissusMind/Materials/MF_TwistedGoo_Preskinned_Glider
graph {nodes [
  n1 = ScalarParameter() {defaultValue=7.1525, parameterName=Fresnel}
  n2 = NamedRerouteUsage()
  n3 = NamedRerouteUsage()
  n4 = AppendVector(
   A=AppendVector(A=ScalarParameter() {defaultValue=84.6447, parameterName=Size_Normal}, B=ScalarParameter() {defaultValue=82.35, parameterName=Size_Normal2})
   B=ScalarParameter() {defaultValue=33.2833, parameterName=Size_Normal3}
  )
  n5 = NamedRerouteUsage()
  n6 = NamedRerouteUsage()
  n7 = Constant()
  n8 = NamedRerouteUsage()
  n9 = NamedRerouteUsage()
  n10 = NamedRerouteUsage()
  n11 = TextureSampleParameter2D() {
   parameterName=BorderFix
   samplerType=SAMPLERTYPE_Masks
   texture=/BRCosmetics/Gadgets/Assets/VinderTech_GliderChute/Glider_MissusMind/Textures/T_Glider_MissusMind_UVBorder
  }
  n12 = MaterialFunctionCall(
   Normal=MaterialFunctionCall(
    Normal=MaterialFunctionCall!(
     Export Float 4
     TextureObject=@n3
     TextureSize=@n4
     WorldPosition=@n5
     World Space Normal=@n6
     ProjectionTransitionContrast=@n7
     Pan_YZ=@n8
     Pan_XZ=@n9
     Pan_XY=@n10
    ) {materialFunction=&/WorldAlignedTexture_ScalePanXYZ}.XYZ Texture
    Flatness=ScalarParameter() {defaultValue=0.25, group=TwistedGoo, parameterName=TG_NormalFlatten, sortPriority=23}
   ) {materialFunction=&FlattenNormal}.Result
   Flatness=@n11.R
  ) {materialFunction=&FlattenNormal}
  n13 = Transform(None=NamedRerouteDeclaration(None=MaterialFunctionCall(BaseNormal=@n2, AdditionalNormal=@n12.Result) {materialFunction=&BlendAngleCorrectedNormals}.Result) {name=Normal, nodeColor=<rgba 0 1 0.5547 1>})
  n14 = DotProduct(A=CameraVectorWS(), B=Normalize(VectorInput=@n13))
  n15 = Multiply(
   A=ScalarParameter() {defaultValue=0.2762, parameterName=FresnelBirghtness2}
   B=Power(Base=@n14, Exp=ScalarParameter() {defaultValue=442.9701, parameterName=Fresnel2})
  )
  Add(
   A=Multiply(A=ScalarParameter() {defaultValue=0.0667, parameterName=FresnelBirghtness}, B=Fresnel(ExponentIn=@n1, Normal=@n13) {baseReflectFraction=0})
   B=@n15
  )
  n16 = ComponentMask(None=MaterialFunctionCall!(
    Export Float 4
    TextureObject=NamedRerouteUsage()
    TextureSize=NamedRerouteUsage()
    WorldPosition=NamedRerouteUsage()
    World Space Normal=NamedRerouteUsage()
    ProjectionTransitionContrast
    Pan_YZ=NamedRerouteUsage()
    Pan_XZ=NamedRerouteUsage()
    Pan_XY=NamedRerouteUsage()
   ) {materialFunction=/Game/Weapons/WeaponSkins/Wraps/MaterialFunctions/WorldAlignedTexture_ComplexPan}.XYZ Texture) {b=true}
  n17 = LinearInterpolate(
   A=ScalarParameter() {defaultValue=-2, group=TwistedGoo, parameterName=TG_VertexOffsetMin}
   B=ScalarParameter() {defaultValue=2, group=TwistedGoo, parameterName=TG_VertexOffsetMax}
   Alpha=@n16
  )
  Clamp(None=@n17) {clampMode=CMODE_ClampMin}
  Constant3Vector() {constant=<rgba 0 0 1 0>}
  DDY()
  n18 = NamedRerouteUsage()
  n19 = NamedRerouteUsage()
  n20 = NamedRerouteUsage()
  FunctionOutput(None=StaticSwitchParameter(True=BlendMaterialAttributes(A=@n18, B=@n19, Alpha=@n20), False=NamedRerouteUsage()) {defaultValue=true, group=TwistedGoo, parameterName=UseTwistedGoo})
  MaterialFunctionCall!(
   TextureObject=NamedRerouteUsage()
   TextureSize=@n4
   Normal=Constant3Vector() {constant=<rgba 0 0 1 0>}
   WorldSpace
   ProjectionTransitionContrast=ScalarParameter() {group=TwistedGoo, parameterName=TG_NormalProjectionContrast, sortPriority=24}
   WorldPosition=Add(A=NamedRerouteUsage(), B=NamedRerouteUsage())
   Use High Quality Normals=StaticBoolParameter() {defaultValue=true, group=TwistedGoo, parameterName=TG_UseHQNormals, sortPriority=26}
   Pan_YZ=NamedRerouteUsage()
   Pan_XZ=NamedRerouteUsage()
   Pan_XY=NamedRerouteUsage()
  ) {materialFunction=/Game/Weapons/WeaponSkins/Wraps/MaterialFunctions/WorldAlignedNormal_ComplexPan}
  Multiply(B=@n14) {constA=-1}
  n21 = PreSkinnedPosition()
  NamedRerouteDeclaration(None=VertexInterpolator(VS=NamedRerouteDeclaration(None=@n21) {name=PreskinnedLocalPosVS, nodeColor=<rgba 1 0 0.0703 1>}).PS) {name=PreskinnedLocalPos, nodeColor=<rgba 1 0 0.0703 1>}
  n22 = PreSkinnedNormal()
  NamedRerouteDeclaration(None=VertexInterpolator(VS=NamedRerouteDeclaration(None=@n22) {name=PreskinnedLocalNormalVS, nodeColor=<rgba 1 0.9609 0 1>}).PS) {name=PreskinnedLocalNormal, nodeColor=<rgba 1 0.9609 0 1>}
  n23 = VectorParameter() {group=TwistedGoo, parameterName=TG_PanDirection_XY, sortPriority=12, defaultValue=<rgba 0.0286 -0.05 0 1>}
  n24 = Normalize(VectorInput=ComponentMask(None=@n23.RGB) {g=true, r=true})
  NamedRerouteDeclaration(None=Multiply(A=@n24, B=NamedRerouteUsage())) {name=Pan_Normal_XY, nodeColor=<rgba 0.7109 1 0 1>}
  NamedRerouteDeclaration(None=Divide(B=ScalarParameter() {defaultValue=1, group=TwistedGoo, parameterName=TG_GlobalTiling, sortPriority=3})) {name=GlobalTiling, nodeColor=<rgba 0 0.4375 1 1>}
  n25 = NamedRerouteUsage()
  NamedRerouteDeclaration(None=Multiply(
    A=ScalarParameter() {defaultValue=6, group=TwistedGoo, parameterName=TG_NormalDistortionStrength, sortPriority=25}
    B=ComponentMask(None=@n25) {g=true}
   )) {name=Distortion, nodeColor=<rgba 0.0078 1 0 1>}
  n26 = TextureObject() {samplerType=SAMPLERTYPE_Masks, texture=/Game/Characters/Player/Male/Medium/Bodies/M_MED_Mastermind/Textures/T_M_MED_Mastermind_Twisted_M}
  NamedRerouteDeclaration(None=@n26) {name=MaskTexture, nodeColor=<rgba 0 0.9297 1 1>}
  NamedRerouteDeclaration(None=Multiply(
    A=ScalarParameter() {defaultValue=80, group=TwistedGoo, parameterName=TG_NormalTextureSize, primitiveDataIndex=10, sortPriority=22}
    B=NamedRerouteUsage()
   )) {name=NormalTextureSize, nodeColor=<rgba 0.6641 0 1 1>}
  NamedRerouteDeclaration(None=Multiply(A=ScalarParameter() {defaultValue=35, group=TwistedGoo, parameterName=TG_DistortionTextureSize}, B=NamedRerouteUsage())) {name=DistortionTextureSize, nodeColor=<rgba 1 0 0.4219 1>}
  n27 = VectorParameter() {group=TwistedGoo, parameterName=TG_VertexOffsetDirection, defaultValue=<rgba 2 -1 2 1>}
  NamedRerouteDeclaration(None=Multiply(A=@n17, B=Transform(None=@n27.RGB) {transformSourceType=TRANSFORMSOURCE_Local})) {name=WPO, nodeColor=<rgba 0 1 0.8594 1>}
  n28 = ScalarParameter() {defaultValue=0.04, group=TwistedGoo, parameterName=TG_DistortionPanSpeed}
  n29 = Multiply(A=@n28) {constB=-1}
  n30 = NamedRerouteUsage()
  n31 = Saturate(None=Sign(None=ComponentMask(None=@n30) {r=true}))
  NamedRerouteDeclaration(None=LinearInterpolate(A=AppendVector(A=@n29, B=@n28), B=AppendVector(A=@n28, B=@n28), Alpha=@n31)) {name=PanSpeedMask_VS, nodeColor=<rgba 1 0 0.375 1>}
  n32 = VectorParameter() {group=TwistedGoo, parameterName=TG_Color_02, sortPriority=2, defaultValue=<rgba 0.0022 0.0017 0.005 1>}
  n33 = NamedRerouteUsage()
  NamedRerouteDeclaration(None=LinearInterpolate(
    A=VectorParameter() {group=TwistedGoo, parameterName=TG_Color_01, sortPriority=1, defaultValue=<rgba 0 0 0 1>}.RGB
    B=@n32.RGB
    Alpha=ComponentMask(None=@n33) {b=true}
   )) {name=BaseColor, nodeColor=<rgba 1 0.1641 0 1>}
  NamedRerouteDeclaration(None=MaterialFunctionCall(
    TextureObject=NamedRerouteUsage()
    TextureSize=NamedRerouteUsage()
    WorldPosition=NamedRerouteUsage()
    World Space Normal=NamedRerouteUsage()
    Pan_YZ=NamedRerouteUsage()
    Pan_XZ=NamedRerouteUsage()
    Pan_XY=NamedRerouteUsage()
   ) {materialFunction=&WorldAlignedTexture_ComplexPan}.XYZ Texture) {name=Mask, nodeColor=<rgba 1 0.8672 0 1>}
  n34 = FunctionInput(Preview=SetMaterialAttributes!(MaterialAttributes)) {bUsePreviewValueAsDefault=true, inputType=FunctionInput_MaterialAttributes}
  NamedRerouteDeclaration(None=GetMaterialAttributes(None=NamedRerouteDeclaration(None=@n34) {name=IN_Material, nodeColor=<rgba 0 1 0.9063 1>}) {attributeGetTypes [#1]}.Normal) {name=IN_Normal, nodeColor=<rgba 0 1 0.4141 1>}
  n35 = SetMaterialAttributes!(
   MaterialAttributes
   BaseColor=NamedRerouteUsage()
   Specular=ScalarParameter() {defaultValue=0.3238, group=TwistedGoo, parameterName=TG_Specular}
   Roughness=ScalarParameter() {defaultValue=0.33, group=TwistedGoo, parameterName=TG_Roughness}
   Normal=NamedRerouteUsage()
   WorldPositionOffset=Constant()
   EmissiveColor=NamedRerouteUsage()
   Metallic=ScalarParameter() {defaultValue=0.33, group=TwistedGoo, parameterName=TG_Metalic}
  ) {attributeSetTypes [
    #2
    #3
    #4
    #1
    #5
    #6
    #7
   ]}
  NamedRerouteDeclaration(None=@n35) {name=OUT_Material, nodeColor=<rgba 0 1 1 1>}
  n36 = FunctionInput() {bUsePreviewValueAsDefault=true, inputName=Alpha, inputType=FunctionInput_Scalar, sortPriority=3, previewValue=<1 0 0 1>}
  NamedRerouteDeclaration(None=@n36) {name=Alpha, nodeColor=<rgba 1 0 0.3984 1>}
  n37 = TextureObject() {samplerType=SAMPLERTYPE_Normal, texture=/Game/Characters/Player/Male/Medium/Bodies/M_MED_Mastermind/Textures/T_M_MED_Mastermind_Twisted_N}
  NamedRerouteDeclaration(None=@n37) {name=NormalTexture, nodeColor=<rgba 1 0 0.2812 1>}
  n38 = ScalarParameter() {defaultValue=0.04, group=TwistedGoo, parameterName=TG_DistortionPanSpeed}
  n39 = Multiply(A=@n38) {constB=-1}
  n40 = NamedRerouteUsage()
  n41 = Saturate(None=Sign(None=ComponentMask(None=@n40) {r=true}))
  NamedRerouteDeclaration(None=LinearInterpolate(A=AppendVector(A=@n39, B=@n38), B=AppendVector(A=@n38, B=@n38), Alpha=@n41)) {name=PanSpeedMask, nodeColor=<rgba 1 0 0.375 1>}
  n42 = ScalarParameter() {defaultValue=0.06, group=TwistedGoo, parameterName=TG_NormalPanSpeed, sortPriority=20}
  n43 = Multiply(A=@n42) {constB=-1}
  NamedRerouteDeclaration(None=LinearInterpolate(A=AppendVector(A=@n43, B=@n42), B=AppendVector(A=@n42, B=@n42), Alpha=@n41)) {name=PanSpeedNormal, nodeColor=<rgba 0 1 1 1>}
  n44 = ScalarParameter() {defaultValue=0.06, group=TwistedGoo, parameterName=TG_NormalPanSpeed, sortPriority=20}
  n45 = Multiply(A=@n44) {constB=-1}
  NamedRerouteDeclaration(None=LinearInterpolate(A=AppendVector(A=@n45, B=@n44), B=AppendVector(A=@n44, B=@n44), Alpha=@n31)) {name=PanSpeedNormal_VS, nodeColor=<rgba 0 1 1 1>}
  n46 = VectorParameter() {group=TwistedGoo, parameterName=TG_PanDirection_YZ, sortPriority=10, defaultValue=<rgba -0.0857 0.1333 0 1>}
  n47 = Normalize(VectorInput=ComponentMask(None=@n46.RGB) {g=true, r=true})
  NamedRerouteDeclaration(None=Multiply(A=NamedRerouteUsage(), B=@n47)) {name=Pan_Mask_YZ_VS, nodeColor=<rgba 0.7109 1 0 1>}
  NamedRerouteDeclaration(None=Multiply(A=@n47, B=NamedRerouteUsage())) {name=Pan_Normal_YZ_VS, nodeColor=<rgba 0.7109 1 0 1>}
  n48 = VectorParameter() {group=TwistedGoo, parameterName=TG_PanDirection_XZ, sortPriority=11, defaultValue=<rgba 0.0762 0.1333 0 1>}
  n49 = Normalize(VectorInput=ComponentMask(None=@n48.RGB) {g=true, r=true})
  NamedRerouteDeclaration(None=Multiply(A=NamedRerouteUsage(), B=@n49)) {name=Pan_Mask_XZ_VS, nodeColor=<rgba 0.7109 1 0 1>}
  n50 = VectorParameter() {group=TwistedGoo, parameterName=TG_PanDirection_XY, sortPriority=12, defaultValue=<rgba 0.0286 -0.05 0 1>}
  n51 = Normalize(VectorInput=ComponentMask(None=@n50.RGB) {g=true, r=true})
  NamedRerouteDeclaration(None=Multiply(A=NamedRerouteUsage(), B=@n51)) {name=Pan_Mask_XY_VS, nodeColor=<rgba 0.7109 1 0 1>}
  NamedRerouteDeclaration(None=Multiply(A=@n49, B=NamedRerouteUsage())) {name=Pan_Normal_XZ_VS, nodeColor=<rgba 0.7109 1 0 1>}
  NamedRerouteDeclaration(None=Multiply(A=@n51, B=NamedRerouteUsage())) {name=Pan_Normal_XY_VS, nodeColor=<rgba 0.7109 1 0 1>}
  n52 = VectorParameter() {group=TwistedGoo, parameterName=TG_PanDirection_YZ, sortPriority=10, defaultValue=<rgba -0.0857 0.1333 0 1>}
  n53 = Normalize(VectorInput=ComponentMask(None=@n52.RGB) {g=true, r=true})
  NamedRerouteDeclaration(None=Multiply(A=NamedRerouteUsage(), B=@n53)) {name=Pan_Mask_YZ, nodeColor=<rgba 0.7109 1 0 1>}
  NamedRerouteDeclaration(None=Multiply(A=@n53, B=NamedRerouteUsage())) {name=Pan_Normal_YZ, nodeColor=<rgba 0.7109 1 0 1>}
  n54 = VectorParameter() {group=TwistedGoo, parameterName=TG_PanDirection_XZ, sortPriority=11, defaultValue=<rgba 0.0762 0.1333 0 1>}
  n55 = Normalize(VectorInput=ComponentMask(None=@n54.RGB) {g=true, r=true})
  NamedRerouteDeclaration(None=Multiply(A=NamedRerouteUsage(), B=@n55)) {name=Pan_Mask_XZ, nodeColor=<rgba 0.7109 1 0 1>}
  NamedRerouteDeclaration(None=Multiply(A=NamedRerouteUsage(), B=@n24)) {name=Pan_Mask_XY, nodeColor=<rgba 0.7109 1 0 1>}
  NamedRerouteDeclaration(None=Multiply(A=@n55, B=NamedRerouteUsage())) {name=Pan_Normal_XZ, nodeColor=<rgba 0.7109 1 0 1>}
  NamedRerouteDeclaration(None=@n15) {name=EM_OUT, nodeColor=<rgba 1 0 0.8672 1>}
  NamedRerouteUsage()
  NamedRerouteUsage()
  NamedRerouteUsage()
  NamedRerouteUsage()
  NamedRerouteUsage()
  NamedRerouteUsage()
  Normalize(VectorInput=Transform(None=NamedRerouteUsage()) {transformSourceType=TRANSFORMSOURCE_Local, transformType=TRANSFORM_Tangent})
  SmoothStep(Value=@n16) {constMax=1.5}
 ]}
